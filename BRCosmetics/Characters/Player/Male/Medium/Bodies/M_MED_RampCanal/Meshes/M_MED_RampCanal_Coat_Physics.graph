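# Physics asset for the RampCanal male medium body's coat, defining capsule bodies and joints for its dynamic coat bones.
# Left and right front, front-side, side, back-side and back chains of four or five segments hang from small offset anchor bodies.
# Tight angular limits, with little or no second swing and twist, generally loosen toward the hem.
# Pelvis, thigh and calf capsules supply body shapes, while a huge zero-length capsule on the floor bone approximates ground.
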PhysicsAsset /BRCosmetics/Characters/Player/Male/Medium/Bodies/M_MED_RampCanal/Meshes/M_MED_RampCanal_Coat_Physics
bodies {
 calf_l [{shapeType=Capsule, length=46.6981, radius=6.2867, center=<-26.6356 -1.2839 0.2544>, rotation=<rot -0.0461 87.6515 90.2612>}]
 calf_r [{shapeType=Capsule, length=46.6981, radius=6.2867, center=<26.5609 0.5378 0.253>, rotation=<rot -0.0461 87.6515 90.2612>}]
 dyn_coat_back_01_lf [{shapeType=Capsule, length=10, radius=5, shapeName=dyn_coat_back_01_rt_capsule, center=<7 -4.3507 0>, rotation=<rot 90 0 0>}]
 dyn_coat_back_01_rt [{shapeType=Capsule, length=10, radius=5, shapeName=dyn_coat_back_01_rt_capsule, center=<7 -4.3507 0>, rotation=<rot 90 0 0>}]
 dyn_coat_back_02_lf [{shapeType=Capsule, length=9, radius=6, shapeName=dyn_coat_back_02_rt_capsule, center=<10 -1.4963 5.6962>}]
 dyn_coat_back_02_rt [{shapeType=Capsule, length=9.5, radius=5.5, shapeName=dyn_coat_back_02_rt_capsule, center=<10 -2.4839 -5.6962>}]
 dyn_coat_back_03_lf [{shapeType=Capsule, length=7, radius=10, shapeName=dyn_coat_back_03_lf_capsule, center=<10 -6.3648 8>}]
 dyn_coat_back_03_rt [{shapeType=Capsule, length=8, radius=10, shapeName=dyn_coat_back_03_lf_capsule, center=<10 -6.3648 -7>, rotation=<rot 0 0.0001 -15>}]
 dyn_coat_back_04_lf [{
   shapeType=Capsule
   length=15.5
   radius=7
   shapeName=dyn_coat_back_04_lf_capsule
   center=<9.8963 -5.097 9.0937>
   rotation=<rot 0 0.0002 20.9819>
  }]
 dyn_coat_back_04_rt [{
   shapeType=Capsule
   length=17
   radius=5.5
   shapeName=dyn_coat_back_04_lf_capsule
   center=<8 -2.8936 -7.1783>
   rotation=<rot 0 0.0004 -15.9818>
  }]
 dyn_coat_back_side_01_lf [{shapeType=Capsule, length=10, radius=5, shapeName=dyn_coat_back_side_01_lf_capsule, center=<11.8553 -5.7741 0>, rotation=<rot 90 0 0>}]
 dyn_coat_back_side_01_rt [{shapeType=Capsule, length=10, radius=5, shapeName=dyn_coat_back_side_01_rt_capsule, center=<10 -5.8116 0>, rotation=<rot 90 0 0>}]
 dyn_coat_back_side_02_lf [{shapeType=Capsule, length=15, radius=9, shapeName=dyn_coat_back_side_02_lf_capsule, center=<7.779 -7.1694 0>}]
 dyn_coat_back_side_02_rt [{shapeType=Capsule, length=15, radius=8, shapeName=dyn_coat_back_side_02_rt_capsule, center=<10 -7.1377 -4>}]
 dyn_coat_back_side_03_lf [{shapeType=Capsule, length=14.5, radius=9, shapeName=dyn_coat_back_side_03_lf_capsule, center=<7.779 -6 0>}]
 dyn_coat_back_side_03_rt [{shapeType=Capsule, length=14.5, radius=9, shapeName=dyn_coat_back_side_03_lf_capsule, center=<10 -6.1703 0>, rotation=<rot -0.0001 0.0001 0>}]
 dyn_coat_back_side_04_lf [{shapeType=Capsule, length=19.5, radius=5.5, shapeName=dyn_coat_back_side_04_lf_capsule, center=<7.5905 -6 0>}]
 dyn_coat_back_side_04_rt [{
   shapeType=Capsule
   length=26
   radius=8
   shapeName=dyn_coat_back_side_04_lf_capsule
   center=<6.3473 -4.0441 0>
   rotation=<rot -0.0001 0.0002 5>
  }]
 dyn_coat_front_01_lf [{shapeType=Capsule, length=7, radius=6.5, shapeName=dyn_coat_front_01_lf_capsule, center=<5.1615 -6.628 0>, rotation=<rot -90 0 0>}]
 dyn_coat_front_01_rt [{shapeType=Capsule, length=11.5, radius=4.5, shapeName=dyn_coat_front_01_rt_capsule, center=<5 -2 0>}]
 dyn_coat_front_02_lf [{
   shapeType=Capsule
   length=6
   radius=5
   shapeName=dyn_coat_front_05_lf_capsule
   center=<7.8614 -3.557 -2.822>
   rotation=<rot -70 89.9996 -89.9998>
  }]
 dyn_coat_front_02_rt [{
   shapeType=Capsule
   length=10.1235
   radius=3
   shapeName=dyn_coat_front_02_rt_capsule
   center=<5.027 -2.0526 0.4868>
   rotation=<rot 78.831 116.74 116.3024>
  }]
 dyn_coat_front_03_lf [{
   shapeType=Capsule
   length=10.4967
   radius=5.5
   shapeName=dyn_coat_front_05_lf_capsule
   center=<8.8285 -0.9815 -3.4497>
   rotation=<rot 0 -0.0001 -20>
  }]
 dyn_coat_front_03_rt [{
   shapeType=Capsule
   length=13
   radius=9
   shapeName=dyn_coat_front_03_lf_capsule
   center=<6.4991 -5.5132 4.4206>
   rotation=<rot -0.0001 -0.0002 10>
  }]
 dyn_coat_front_04_lf [{
   shapeType=Capsule
   length=6.255
   radius=6.5
   shapeName=dyn_coat_front_05_lf_capsule
   center=<4.7309 -0.6924 -5.4153>
   rotation=<rot 0 -0.0001 -20>
  }]
 dyn_coat_front_04_rt [{
   shapeType=Capsule
   length=14.5
   radius=10
   shapeName=dyn_coat_front_03_lf_capsule
   center=<7.4388 -3.6258 5.5111>
   rotation=<rot -0.0001 -0.0002 25>
  }]
 dyn_coat_front_05_lf [{
   shapeType=Capsule
   length=3.7861
   radius=10
   shapeName=dyn_coat_front_05_lf_capsule
   center=<10.8895 -0.2606 -4.5456>
   rotation=<rot 0 -0.0001 -30>
  }]
 dyn_coat_front_05_rt [{
   shapeType=Capsule
   length=5.7183
   radius=8
   shapeName=dyn_coat_front_05_lf_capsule
   center=<7.2132 -0.3026 6.449>
   rotation=<rot 0 -0.0001 30>
  }]
 dyn_coat_front_side_01_lf [{
   shapeType=Capsule
   length=12.0311
   radius=7.987
   shapeName=dyn_coat_front_side_01_rt_capsule
   center=<14.0067 -6.5981 0.7344>
   rotation=<rot -0.2868 -83.4593 -86.7206>
  }]
 dyn_coat_front_side_01_rt [{
   shapeType=Capsule
   length=10.2425
   radius=6.987
   shapeName=dyn_coat_front_side_01_rt_capsule
   center=<10.8493 -5.9805 0.915>
   rotation=<rot 0 -88.4511 -86.708>
  }]
 dyn_coat_front_side_02_lf [{
   shapeType=Capsule
   length=22.4909
   radius=6.5091
   shapeName=dyn_coat_front_side_02_lf_capsule
   center=<10.4256 -5.0225 -0.1138>
   rotation=<rot 0 0 5>
  }]
 dyn_coat_front_side_02_rt [{
   shapeType=Capsule
   length=20.9909
   radius=8.0091
   shapeName=dyn_coat_front_side_02_lf_capsule
   center=<10.4256 -7.0854 -0.5175>
   rotation=<rot 0 0 -5>
  }]
 dyn_coat_front_side_03_lf [{shapeType=Capsule, length=23.6295, radius=7.3705, shapeName=dyn_coat_front_side_03_lf_capsule, center=<8.0463 -5.9619 0>}]
 dyn_coat_front_side_03_rt [{shapeType=Capsule, length=31.1295, radius=7.8705, shapeName=dyn_coat_front_side_03_lf_capsule, center=<8.0463 -7.8178 0>}]
 dyn_coat_front_side_04_lf [
  {shapeType=Capsule, length=30.5585, radius=6.2113, shapeName=dyn_coat_front_side_04_rt_capsule, center=<7.2951 -1.9692 0>}
  {shapeType=Capsule, length=8.5, radius=5, shapeName=dyn_coat_front_side_04_lf_capsule, center=<7.9745 0.3031 0>, rotation=<rot -90 -0.0001 0>}
 ]
 dyn_coat_front_side_04_rt [{shapeType=Capsule, length=30.5585, radius=6.2113, shapeName=dyn_coat_front_side_04_rt_capsule, center=<7.2951 -1.9692 0>}]
 dyn_coat_side_01_lf [{shapeType=Capsule, length=8.8105, radius=4.5, shapeName=dyn_coat_side_01_lf_capsule, center=<10 -4.4302 0>, rotation=<rot 90 0 0>}]
 dyn_coat_side_01_rt [{shapeType=Capsule, length=9.3871, radius=4.5, shapeName=dyn_coat_side_01_rt_capsule, center=<11.7183 -3.4884 0>, rotation=<rot 90 0 0>}]
 dyn_coat_side_02_lf [{shapeType=Capsule, length=17.5, radius=7, shapeName=dyn_coat_side_02_lf_capsule, center=<8.2671 -5.4479 0>}]
 dyn_coat_side_02_rt [{shapeType=Capsule, length=11.5, radius=6, shapeName=dyn_coat_side_02_lf_capsule, center=<10 -3.8583 -0.0629>, rotation=<rot 0 0 10>}]
 dyn_coat_side_03_lf [{shapeType=Capsule, length=16.5, radius=7, shapeName=dyn_coat_side_03_lf_capsule, center=<7.2133 -5.4479 0>}]
 dyn_coat_side_03_rt [{shapeType=Capsule, length=14.5, radius=8, shapeName=dyn_coat_side_03_lf_capsule, center=<10 -4.4023 0>}]
 dyn_coat_side_04_lf [{shapeType=Capsule, length=23, radius=7, shapeName=dyn_coat_side_04_lf_capsule, center=<5.6849 -3.3474 0>}]
 dyn_coat_side_04_rt [{
   shapeType=Capsule
   length=20.5
   radius=7
   shapeName=dyn_coat_side_04_lf_capsule
   center=<5.6849 -1.8862 -0.411>
   rotation=<rot -0.0001 0.0001 5>
  }]
 floor [{shapeType=Capsule, length=0, radius=396.6109, shapeName=root_capsule, center=<0 0 -390.6006>}]
 offset_coat_back_lf [{shapeType=Capsule, length=0.505, radius=0.505, shapeName=offset_coat_back_lf_capsule, rotation=<rot 0 -90 0>}]
 offset_coat_back_rt [{shapeType=Capsule, length=0.505, radius=0.505, shapeName=offset_coat_back_rt_capsule, rotation=<rot 0 -90 0>}]
 offset_coat_back_side_lf [{shapeType=Capsule, length=0.505, radius=0.505, shapeName=offset_coat_back_side_lf_capsule, rotation=<rot 0 -90 0>}]
 offset_coat_back_side_rt [{shapeType=Capsule, length=0.505, radius=0.505, shapeName=offset_coat_back_side_rt_capsule, rotation=<rot 0 -90 0>}]
 offset_coat_front_lf [{shapeType=Capsule, length=0.505, radius=0.505, shapeName=offset_coat_front_lf_capsule, rotation=<rot 0 -90 0>}]
 offset_coat_front_rt [{shapeType=Capsule, length=0.505, radius=0.505, shapeName=offset_coat_front_rt_capsule, rotation=<rot 0 -90 0>}]
 offset_coat_front_side_lf [{shapeType=Capsule, length=0.505, radius=0.505, shapeName=offset_coat_front_side_lf_capsule, rotation=<rot 0 -90 0>}]
 offset_coat_front_side_rt [{shapeType=Capsule, length=0.505, radius=0.505, shapeName=offset_coat_front_side_rt_capsule, rotation=<rot 0 -90 0>}]
 offset_coat_side_lf [{shapeType=Capsule, length=0.505, radius=0.505, shapeName=offset_coat_side_lf_capsule, rotation=<rot 0 -90 0>}]
 offset_coat_side_rt [{shapeType=Capsule, length=0.505, radius=0.505, shapeName=offset_coat_side_rt_capsule, rotation=<rot 0 -90 0>}]
 pelvis [{shapeType=Capsule, length=12.4626, radius=9.5991, center=<-2.933 0.4838 0>, rotation=<rot 0 -90 0>}]
 thigh_l [{shapeType=Capsule, length=29.2797, radius=7.8206, center=<-23.078 0.9595 0.5463>, rotation=<rot -0.4094 -88.7186 -87.3226>}]
 thigh_r [{shapeType=Capsule, length=24.7797, radius=8.3206, center=<17.2668 -0.1969 -1.3302>, rotation=<rot -0.4094 -88.7188 92.6774>}]
}
constraints [
 {bone1Name=dyn_coat_back_02_lf, bone2Name=dyn_coat_back_01_lf, swing1LimitDegrees=10, swing2LimitDegrees=0, twistLimitDegrees=0}
 {bone1Name=dyn_coat_back_03_lf, bone2Name=dyn_coat_back_02_lf, swing1LimitDegrees=15, swing2LimitDegrees=0, twistLimitDegrees=0}
 {bone1Name=dyn_coat_back_04_lf, bone2Name=dyn_coat_back_03_lf, swing1LimitDegrees=20, swing2LimitDegrees=5, twistLimitDegrees=0}
 {bone1Name=dyn_coat_back_02_rt, bone2Name=dyn_coat_back_01_rt, swing1LimitDegrees=10, swing2LimitDegrees=0, twistLimitDegrees=0}
 {bone1Name=dyn_coat_back_03_rt, bone2Name=dyn_coat_back_02_rt, swing1LimitDegrees=15, swing2LimitDegrees=0, twistLimitDegrees=0}
 {bone1Name=dyn_coat_back_04_rt, bone2Name=dyn_coat_back_03_rt, swing1LimitDegrees=20, swing2LimitDegrees=5, twistLimitDegrees=0}
 {bone1Name=dyn_coat_back_side_02_lf, bone2Name=dyn_coat_back_side_01_lf, swing1LimitDegrees=10, swing2LimitDegrees=0, twistLimitDegrees=0}
 {bone1Name=dyn_coat_back_side_03_lf, bone2Name=dyn_coat_back_side_02_lf, swing1LimitDegrees=20, swing2LimitDegrees=0, twistLimitDegrees=5}
 {bone1Name=dyn_coat_back_side_04_lf, bone2Name=dyn_coat_back_side_03_lf, swing1LimitDegrees=20, swing2LimitDegrees=5, twistLimitDegrees=5}
 {bone1Name=dyn_coat_back_side_02_rt, bone2Name=dyn_coat_back_side_01_rt, swing1LimitDegrees=10, swing2LimitDegrees=0, twistLimitDegrees=0}
 {bone1Name=dyn_coat_back_side_03_rt, bone2Name=dyn_coat_back_side_02_rt, swing1LimitDegrees=20, swing2LimitDegrees=0, twistLimitDegrees=5}
 {bone1Name=dyn_coat_back_side_04_rt, bone2Name=dyn_coat_back_side_03_rt, swing1LimitDegrees=20, swing2LimitDegrees=5, twistLimitDegrees=5}
 {
  bone1Name=dyn_coat_front_side_01_lf
  bone2Name=offset_coat_front_side_lf
  swing1LimitDegrees=5
  swing2LimitDegrees=0
  twistLimitDegrees=45
  twistMotion=Locked
 }
 {bone1Name=dyn_coat_front_side_02_lf, bone2Name=dyn_coat_front_side_01_lf, swing1LimitDegrees=10, swing2LimitDegrees=0, twistLimitDegrees=0}
 {bone1Name=dyn_coat_front_side_03_lf, bone2Name=dyn_coat_front_side_02_lf, swing1LimitDegrees=10, swing2LimitDegrees=0, twistLimitDegrees=0}
 {bone1Name=dyn_coat_front_side_04_lf, bone2Name=dyn_coat_front_side_03_lf, swing1LimitDegrees=10, swing2LimitDegrees=0, twistLimitDegrees=10}
 {
  bone1Name=dyn_coat_front_side_01_rt
  bone2Name=offset_coat_front_side_rt
  swing1LimitDegrees=5
  swing2LimitDegrees=0
  twistLimitDegrees=45
  twistMotion=Locked
 }
 {
  bone1Name=dyn_coat_front_side_02_rt
  bone2Name=dyn_coat_front_side_01_rt
  swing1LimitDegrees=10
  swing2LimitDegrees=0
  twistLimitDegrees=45
  twistMotion=Locked
 }
 {bone1Name=dyn_coat_front_side_03_rt, bone2Name=dyn_coat_front_side_02_rt, swing1LimitDegrees=15, swing2LimitDegrees=0, twistLimitDegrees=5}
 {bone1Name=dyn_coat_front_side_04_rt, bone2Name=dyn_coat_front_side_03_rt, swing1LimitDegrees=25, swing2LimitDegrees=10, twistLimitDegrees=5}
 {
  bone1Name=dyn_coat_side_02_lf
  bone2Name=dyn_coat_side_01_lf
  swing1LimitDegrees=10
  swing2LimitDegrees=0
  twistLimitDegrees=45
  twistMotion=Locked
 }
 {bone1Name=dyn_coat_side_03_lf, bone2Name=dyn_coat_side_02_lf, swing1LimitDegrees=15, swing2LimitDegrees=5, twistLimitDegrees=0}
 {bone1Name=dyn_coat_side_04_lf, bone2Name=dyn_coat_side_03_lf, swing1LimitDegrees=20, swing2LimitDegrees=5, twistLimitDegrees=10}
 {
  bone1Name=dyn_coat_side_02_rt
  bone2Name=dyn_coat_side_01_rt
  swing1LimitDegrees=10
  swing2LimitDegrees=0
  twistLimitDegrees=45
  twistMotion=Locked
 }
 {bone1Name=dyn_coat_side_03_rt, bone2Name=dyn_coat_side_02_rt, swing1LimitDegrees=15, swing2LimitDegrees=5, twistLimitDegrees=0}
 {bone1Name=dyn_coat_side_04_rt, bone2Name=dyn_coat_side_03_rt, swing1LimitDegrees=20, swing2LimitDegrees=5, twistLimitDegrees=0}
 {
  bone1Name=dyn_coat_front_03_lf
  bone2Name=dyn_coat_front_02_lf
  swing1LimitDegrees=10
  swing2LimitDegrees=0
  twistLimitDegrees=45
  twistMotion=Locked
 }
 {bone1Name=dyn_coat_front_04_lf, bone2Name=dyn_coat_front_03_lf, swing1LimitDegrees=10, swing2LimitDegrees=0, twistLimitDegrees=0}
 {bone1Name=dyn_coat_front_05_lf, bone2Name=dyn_coat_front_04_lf, swing1LimitDegrees=20, swing2LimitDegrees=0, twistLimitDegrees=10}
 {bone1Name=dyn_coat_front_03_rt, bone2Name=dyn_coat_front_02_rt, swing1LimitDegrees=20, swing2LimitDegrees=0, twistLimitDegrees=10}
 {bone1Name=dyn_coat_front_04_rt, bone2Name=dyn_coat_front_03_rt, swing1LimitDegrees=15, swing2LimitDegrees=0, twistLimitDegrees=10}
 {bone1Name=dyn_coat_front_05_rt, bone2Name=dyn_coat_front_04_rt, swing1LimitDegrees=15, swing2LimitDegrees=5, twistLimitDegrees=0}
 {bone1Name=dyn_coat_front_02_rt, bone2Name=dyn_coat_front_01_rt, swing1LimitDegrees=15, swing2LimitDegrees=0, twistLimitDegrees=0}
 {
  bone1Name=dyn_coat_front_01_rt
  bone2Name=offset_coat_front_rt
  swing1LimitDegrees=5
  swing2LimitDegrees=0
  twistLimitDegrees=45
  twistMotion=Locked
 }
 {bone1Name=dyn_coat_back_01_lf, bone2Name=offset_coat_back_lf, swing1LimitDegrees=5, swing2LimitDegrees=0, twistLimitDegrees=0}
 {bone1Name=dyn_coat_back_01_rt, bone2Name=offset_coat_back_rt, swing1LimitDegrees=5, swing2LimitDegrees=0, twistLimitDegrees=0}
 {bone1Name=dyn_coat_back_side_01_lf, bone2Name=offset_coat_back_side_lf, swing1LimitDegrees=5, swing2LimitDegrees=0, twistLimitDegrees=0}
 {bone1Name=dyn_coat_back_side_01_rt, bone2Name=offset_coat_back_side_rt, swing1LimitDegrees=5, swing2LimitDegrees=0, twistLimitDegrees=0}
 {
  bone1Name=dyn_coat_side_01_lf
  bone2Name=offset_coat_side_lf
  swing1LimitDegrees=5
  swing2LimitDegrees=0
  twistLimitDegrees=45
  twistMotion=Locked
 }
 {
  bone1Name=dyn_coat_side_01_rt
  bone2Name=offset_coat_side_rt
  swing1LimitDegrees=5
  swing2LimitDegrees=0
  twistLimitDegrees=45
  twistMotion=Locked
 }
 {bone1Name=dyn_coat_front_02_lf, bone2Name=dyn_coat_front_01_lf, swing1LimitDegrees=12, swing2LimitDegrees=0, twistLimitDegrees=10}
 {
  bone1Name=dyn_coat_front_01_lf
  bone2Name=offset_coat_front_lf
  swing1LimitDegrees=10
  swing2LimitDegrees=0
  twistLimitDegrees=45
  twistMotion=Locked
 }
]
