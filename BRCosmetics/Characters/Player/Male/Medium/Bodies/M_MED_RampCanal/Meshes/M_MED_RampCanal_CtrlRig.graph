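# Control Rig for the M_MED_RampCanal player body that procedurally corrects body deformation, hood and coat bones.
# It runs the shared DeformRig function, constrains hood ropes, and offsets the hood back from head pose during crouch.
# With cloth off, thigh spherical pose readers, often scaled by calf-to-target distance, drive coat offsets, alongside smoothed crouch, melee and rail adjustments.
# With cloth on, it instead aims a pelvis cloth helper at root unless airborne.
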ControlRigBlueprint /BRCosmetics/Characters/Player/Male/Medium/Bodies/M_MED_RampCanal/Meshes/M_MED_RampCanal_CtrlRig
graphs [
 New Function_ContainedGraph {nodes [
   n1 = RigVMFunctionEntryNode()
   RigVMFunction_MathBoolToInteger = &RigVMFunction_MathBoolToInteger(Value=@n1.onlyTwist)
   RigVMFunction_MathQuaternionSwingTwist = &RigVMFunction_MathQuaternionSwingTwist(
    twistAxis=@n1.twistAxis
    Input=&RigUnit_GetRelativeTransformForItem(
     bChildInitial=False
     bParentInitial=False
     child=(Type=Bone,Name="None")
     parent=(Type=Bone,Name="None")
     child.Name=@n1.child
     parent.Name=@n1.parent
     parent.Type=@n1.parentType
    ).RelativeTransform.Rotation
   )
   RigVMFunction_MathQuaternionToEuler = &RigVMFunction_MathQuaternionToEuler(
    Value=RigVMDispatchNode(
     Values=((),())
     Index=@RigVMFunction_MathBoolToInteger.Result
     Values.0=@RigVMFunction_MathQuaternionSwingTwist.Swing
     Values.1=@RigVMFunction_MathQuaternionSwingTwist.Twist
    ) {template="DISPATCH_RigVMDispatch_SelectInt32(in Index,in Values,out Result)"}.Result
    RotationOrder=@n1.Rotation Order
   )
   > DISPATCH_RigVMDispatch_Print = RigVMDispatchNode(ScreenColor=(R=1,G=1,B=1,A=1), ScreenDuration=1.5, Enabled=@n1.debugRotation, Value=@RigVMFunction_MathQuaternionToEuler.Result) {template="DISPATCH_RigVMDispatch_Print(in Prefix,in Value,in Enabled,in ScreenDuration,in ScreenColor)"}
   Clamp = &RigVMFunction_MathIntClamp(Value=@n1.primAxis)
   DISPATCH_RigVMDispatch_SelectInt32 = RigVMDispatchNode(
    Values=(0,0,0)
    Index=@Clamp.Result
    Values.0=@RigVMFunction_MathQuaternionToEuler.Result.X
    Values.1=@RigVMFunction_MathQuaternionToEuler.Result.Y
    Values.2=@RigVMFunction_MathQuaternionToEuler.Result.Z
   ) {template="DISPATCH_RigVMDispatch_SelectInt32(in Index,in Values,out Result)"}
   Remap = &RigVMFunction_MathDoubleRemap(
    TargetMaximum=1
    TargetMinimum=0
    bClamp=False
    Value=@DISPATCH_RigVMDispatch_SelectInt32.Result
    SourceMaximum=@n1.max
    SourceMinimum=@n1.min
   )
   Multiply = &RigVMFunction_MathDoubleMul(A=&RigVMFunction_MathDoubleClamp(Value=@Remap.Result).Result, B=@n1.premult)
   > RigVMFunctionReturnNode(outputWeight=@Multiply.Result)
  ]}
 RigVMFunctionLibrary {nodes [DeltaRotation = RigVMCollapseNode(ExecuteContext=(), Rotation Order=ZYX, premult=1, twistAxis=(X=1,Y=0,Z=0))]}
 RigVMFunction_Sequence_1_ContainedGraph {nodes [
   n1 = RigVMFunctionEntryNode()
   RigVMFunction_Sequence = &RigVMFunction_Sequence(ExecuteContext=@n1.ExecuteContext)
   n2 = &RigVMFunction_Sequence(ExecuteContext=@RigVMFunction_Sequence.B)
   RigVMFunctionReturnNode(A=@RigVMFunction_Sequence.A, B=@n2.A, C=@n2.B)
  ]}
 RigVMFunction_Sequence_2_ContainedGraph {nodes [
   n1 = RigVMFunctionEntryNode()
   n2 = &RigVMFunction_Sequence(ExecuteContext=@n1.ExecuteContext)
   n3 = &RigVMFunction_Sequence(ExecuteContext=@n2.B)
   RigVMFunctionReturnNode(A=@n2.A, B=@n3.A, C=@n3.B)
  ]}
 RigVMModel {nodes [
   RigUnit_BeginExecution = &RigUnit_BeginExecution()
   RigVMFunction_Sequence = RigVMAggregateNode(ExecuteContext=@RigUnit_BeginExecution.ExecutePin)
   FN_DeformRig_FNC = RigVMFunctionReferenceNode(
    DeformRig_ Curve="(deformRig_l_upperarm_up_alpha_72_#1=False,deformRig_r_upperarm_up_alpha_73_#2=False,deformRig_l_upperarm_fr_alpha_74_#3=False,deformRig_r_upperarm_fr_alpha_75_CB51B8 ...(truncated)"
    DeformRig_Bool="(upperarm_1_#4=True,elbow_3_#5=True,wrist_7_#6=True,groin_8_#7=True,glute_17_#8=True,knee_9_40C408324CA261 ...(truncated)"
    DeformRig_Debug=(upperarm_debug_18_#4=False,wrist_debug_20_#6=False,lat_debug_26_#9=False,groinGlute_debug_27_#7=False)
    DeformRig_Value="(upperarm_up_18_#10=(Rotation=(X=0.0,Y=-0.0,Z=0.0,W=1.0),Translation=(X=0.0,Y=0.0,Z=8.0),Scale3D=(X=0.70,Y=1.0,Z=1.0)),upperarm_fwd_fr_19_#11=(Rotation=(X=0.0,Y=-0.0,Z=0.0,W=1.0),Translation=(X=0.0, ...(truncated)"
    ExecuteContext=@RigVMFunction_Sequence.A
   )
   > &RigUnit_ParentConstraint(
    AdvancedSettings=(RotationOrderForFilter=XZY)
    Child=(Type=Bone,Name="base_hood_rope_front_01_rt")
    Parents=((Item=(Type=Bone,Name="spine_03"),Weight=0.0461),(Item=(Type=Bone,Name="deform_pec_r"),Weight=0.0472),(Item=(Type=Bone,Name="spine_04"),Weight=0.1499),(Item=(Type=Bone,Name="clavicle_r"),Weight=0.2077),(Item=(Type=Bone,Name="spine_05"),Weight=0.549))
    Weight=1
   )
   > &RigUnit_ParentConstraint(
    AdvancedSettings=(RotationOrderForFilter=XZY)
    Child=(Type=Bone,Name="base_hood_rope_front_01_lf")
    Parents="((Item=(Type=Bone,Name=\"spine_02\"),Weight=0.0126),(Item=(Type=Bone,Name=\"deform_upperarm_fr_l\"),Weight=0.0154),(Item=(Type=Bone,Name=\"upperarm_twist_01_l\"),Weight=0.0255),(Item=(Type=Bone,Name=\"spine_03\"),Weight=0.0465),(Item=(Type=Bone,Name=\"deform_pec_l\" ...(truncated)"
    Weight=1
   )
   > &RigUnit_ParentConstraint(
    AdvancedSettings=(RotationOrderForFilter=XZY)
    Child=(Type=Bone,Name="aim_hood_rope_front_01_lf")
    Parents="((Item=(Type=Bone,Name=\"spine_05\"),Weight=0.0298),(Item=(Type=Bone,Name=\"spine_01\"),Weight=0.0476),(Item=(Type=Bone,Name=\"clavicle_l\"),Weight=0.0915),(Item=(Type=Bone,Name=\"spine_02\"),Weight=0.1516),(Item=(Type=Bone,Name=\"deform_pec_l\"),Weight=0.152),(Item ...(truncated)"
    Weight=1
   )
   > &RigUnit_ParentConstraint(
    AdvancedSettings=(RotationOrderForFilter=XZY)
    Child=(Type=Bone,Name="aim_hood_rope_front_01_rt")
    Parents="((Item=(Type=Bone,Name=\"spine_05\"),Weight=0.0341),(Item=(Type=Bone,Name=\"deform_pec_l\"),Weight=0.0366),(Item=(Type=Bone,Name=\"spine_01\"),Weight=0.0479),(Item=(Type=Bone,Name=\"clavicle_r\"),Weight=0.0805),(Item=(Type=Bone,Name=\"deform_pec_r\"),Weight=0.0903), ...(truncated)"
    Weight=1
   )
   > &RigUnit_AimConstraintLocalSpaceOffset(
    Child=(Type=Bone,Name="base_hood_rope_front_01_lf")
    Filter=(bX=True,bY=True,bZ=True)
    Parents=((Item=(Type=Bone,Name="aim_hood_rope_front_01_lf"),Weight=1))
    UpAxis=(X=0,Y=-1,Z=0)
    Weight=1
    WorldUp=(Target=(X=0,Y=-1,Z=0),Kind=Direction,Space=(Type=Bone,Name="aim_hood_rope_front_01_lf"))
   )
   > &RigUnit_AimConstraintLocalSpaceOffset(
    Child=(Type=Bone,Name="base_hood_rope_front_01_rt")
    Filter=(bX=True,bY=True,bZ=True)
    Parents=((Item=(Type=Bone,Name="aim_hood_rope_front_01_rt"),Weight=1))
    UpAxis=(X=0,Y=-1,Z=0)
    Weight=1
    WorldUp=(Target=(X=0,Y=-1,Z=0),Kind=Direction,Space=(Type=Bone,Name="aim_hood_rope_front_01_rt"))
   )
   RigVMFunction_MathBoolOr = &RigVMFunction_MathBoolOr(A=false, B=False)
   n1 = &RigUnit_SphericalPoseReader(
    ActiveRegionScaleFactors=(PositiveWidth=1,NegativeWidth=0.5,PositiveHeight=0.75,NegativeHeight=0.75)
    ActiveRegionSize=0.1
    Debug=(bDrawDebug=true,bDraw2D=False,bDrawLocalAxes=False,DebugScale=25,DebugSegments=20,DebugThickness=0.25)
    DriverAxis=(X=0,Y=1,Z=0)
    DriverItem=(Type=Bone,Name="head")
    FalloffRegionScaleFactors=(PositiveWidth=1,NegativeWidth=0.1,PositiveHeight=1,NegativeHeight=1)
    FalloffSize=0.2
    FlipHeightScaling=False
    FlipWidthScaling=False
    OptionalParentItem=(Type=Bone,Name="None")
    RotationOffset=(X=100,Y=0,Z=-40)
    Debug.bDrawDebug=@RigVMFunction_MathBoolOr.Result
    ExecutePin=@RigVMFunction_Sequence.B
   )
   n2 = &RigVMFunction_AnimEvalRichCurve(
    Curve=(EditorCurveData=(Keys=((),(Time=0.5,Value=1),(Time=0.8))))
    SourceMaximum=2
    SourceMinimum=0
    TargetMinimum=0
    Value=&RigVMFunction_MathDoubleRemap(
     SourceMaximum=70
     SourceMinimum=45
     TargetMaximum=1
     TargetMinimum=0
     Value=&RigUnit_GetRelativeTransformForItem(Child=(Type=Bone,Name="pelvis"), Parent=(Type=Bone,Name="root"), bChildInitial=False, bParentInitial=False).RelativeTransform.Translation.Z
    ).Result
   )
   n3 = RigVMVariableNode() {variable=CRisCrouching}
   Multiply = &RigVMFunction_MathDoubleMul(
    A=@n1.OutputParam
    B=&RigVMFunction_MathDoubleMul(B=@n2.Result, A=&RigVMFunction_MathBoolToFloat(Value=@n3.Value).Result).Result
   )
   AlphaInterp = &RigVMFunction_AlphaInterp(InterpSpeedDecreasing=2, InterpSpeedIncreasing=10, Scale=1, bClampResult=true, bMapRange=true, Value=@Multiply.Result)
   > &RigUnit_ModifyTransforms(
    ItemToModify=((Item=(Type=Bone,Name="base_hood_back_cr_top"),Transform=(Rotation=(X=0,Y=0,Z=0.2588,W=0.9659),Translation=(X=12,Y=5,Z=0),Scale3D=(X=1,Y=1,Z=1))))
    WeightMinimum=0
    Weight=@AlphaInterp.Result
   )
   n4 = &RigVMFunction_ControlFlowBranch(Condition=RigVMVariableNode() {variable=IsClothOn}.Value, ExecuteContext=@RigVMFunction_Sequence.C)
   RigVMFunction_Sequence_1 = RigVMAggregateNode(ExecuteContext=@n4.False)
   n5 = &RigUnit_SphericalPoseReader(
    ActiveRegionScaleFactors=(PositiveWidth=1,NegativeWidth=1,PositiveHeight=1,NegativeHeight=1)
    ActiveRegionSize=0.4
    Debug=(bDrawDebug=false,bDraw2D=False,bDrawLocalAxes=False,DebugScale=25,DebugSegments=20,DebugThickness=0.25)
    DriverAxis=(X=-1,Y=0,Z=0)
    DriverItem=(Type=Bone,Name="thigh_l")
    FalloffRegionScaleFactors=(PositiveWidth=1,NegativeWidth=1,PositiveHeight=1,NegativeHeight=1)
    FalloffSize=0.6
    FlipHeightScaling=False
    FlipWidthScaling=False
    OptionalParentItem=(Type=Bone,Name="None")
    RotationOffset=(X=90,Y=0,Z=-20)
    ExecutePin=@RigVMFunction_Sequence_1.A
   )
   RigUnit_GetTransform = &RigUnit_GetTransform(Item=(Type=Bone,Name="calf_l"), bInitial=False)
   n6 = &RigUnit_GetTransform(Item=(Type=Bone,Name="tgt_coat_front_lf"), bInitial=False)
   n7 = &RigVMFunction_MathDoubleMul(
    B=&RigVMFunction_MathDoubleRemap(
     SourceMaximum=11.6097
     SourceMinimum=60.629
     TargetMaximum=1
     TargetMinimum=0
     Value=&RigVMFunction_MathVectorDistance(B=@RigUnit_GetTransform.Transform.Translation, A=@n6.Transform.Translation).Result
    ).Result
    A=@n5.OutputParam
   )
   > &RigUnit_ModifyTransforms(
    ItemToModify="((Item=(Type=Bone,Name=\"orient_coat_front_lf\"),Transform=(Rotation=(X=0,Y=0,Z=0,W=1),Translation=(X=0,Y=-2.6669,Z=-3.907),Scale3D=(X=1,Y=1,Z=1))),(Item=(Type=Bone,Name=\"offset_coat_front_lf\"),Transform=(Rotation=(X=0.3582,Y=-0.2376,Z=0.4389,W=0.789),Transl ...(truncated)"
    WeightMinimum=0
    Weight=@n7.Result
   )
   n8 = &RigVMFunction_AlphaInterp(
    InterpSpeedDecreasing=5
    InterpSpeedIncreasing=100
    Scale=1
    bClampResult=true
    Value=&RigVMFunction_MathBoolToFloat(Value=RigVMVariableNode() {variable=IsCrouchWalking}.Value).Result
   )
   > &RigUnit_ModifyTransforms(
    ItemToModify="((Item=(Type=Bone,Name=\"orient_coat_front_lf\"),Transform=(Rotation=(X=0,Y=0,Z=0,W=1),Translation=(X=0,Y=0,Z=0),Scale3D=(X=1,Y=1,Z=1))),(Item=(Type=Bone,Name=\"offset_coat_front_lf\"),Transform=(Rotation=(X=0,Y=0,Z=0,W=1),Translation=(X=0,Y=0,Z=0),Scale3D=(X= ...(truncated)"
    WeightMinimum=0
    Weight=&RigVMFunction_MathDoubleMul(A=@n8.Result, B=@n7.Result).Result
   )
   > n9 = &RigUnit_SphericalPoseReader(
    ActiveRegionScaleFactors=(PositiveWidth=0.5,NegativeWidth=1,PositiveHeight=0.5,NegativeHeight=0.5)
    ActiveRegionSize=0.35
    Debug=(bDrawDebug=false,bDraw2D=False,bDrawLocalAxes=False,DebugScale=25,DebugSegments=20,DebugThickness=0.25)
    DriverAxis=(X=-1,Y=0,Z=0)
    DriverItem=(Type=Bone,Name="thigh_l")
    FalloffRegionScaleFactors=(PositiveWidth=0.5,NegativeWidth=1,PositiveHeight=0.5,NegativeHeight=0.5)
    FalloffSize=0.55
    FlipHeightScaling=False
    FlipWidthScaling=False
    OptionalParentItem=(Type=Bone,Name="None")
    RotationOffset=(X=180,Y=20,Z=0)
   )
   n10 = &RigVMFunction_MathVectorDistance(
    B=&RigUnit_GetTransform(Item=(Type=Bone,Name="calf_l"), bInitial=False).Transform.Translation
    A=&RigUnit_GetTransform(Item=(Type=Bone,Name="tgt_coat_side_lf"), bInitial=False).Transform.Translation
   )
   > &RigUnit_ModifyTransforms(
    ItemToModify="((Item=(Type=Bone,Name=\"offset_coat_back_side_lf\"),Transform=(Rotation=(X=-0.0915,Y=0.4126,Z=0.1962,W=-0.8848),Translation=(X=0,Y=0,Z=0),Scale3D=(X=1,Y=1,Z=1))),(Item=(Type=Bone,Name=\"offset_coat_front_lf\"),Transform=(Rotation=(X=0.0131,Y=-0.0416,Z=-0.3004 ...(truncated)"
    WeightMinimum=0
    Weight=&RigVMFunction_MathDoubleMul(
     B=&RigVMFunction_MathDoubleRemap(SourceMaximum=22, SourceMinimum=58.69, TargetMaximum=1, TargetMinimum=0, Value=@n10.Result).Result
     A=@n9.OutputParam
    ).Result
   )
   > n11 = &RigUnit_SphericalPoseReader(
    ActiveRegionScaleFactors=(PositiveWidth=1,NegativeWidth=1,PositiveHeight=0.5,NegativeHeight=0.5)
    ActiveRegionSize=0.5
    Debug=(bDrawDebug=false,bDraw2D=False,bDrawLocalAxes=False,DebugScale=25,DebugSegments=20,DebugThickness=0.25)
    DriverAxis=(X=-1,Y=0,Z=0)
    DriverItem=(Type=Bone,Name="thigh_l")
    FalloffRegionScaleFactors=(PositiveWidth=0.5,NegativeWidth=1,PositiveHeight=1,NegativeHeight=0.7)
    FalloffSize=0.58
    FlipHeightScaling=False
    FlipWidthScaling=False
    OptionalParentItem=(Type=Bone,Name="None")
    RotationOffset=(X=130,Y=0,Z=-20)
   )
   n12 = &RigUnit_GetTransform(Item=(Type=Bone,Name="calf_l"), bInitial=False)
   n13 = &RigUnit_GetTransform(Item=(Type=Bone,Name="tgt_coat_front_side_lf"), bInitial=False)
   n14 = &RigVMFunction_MathDoubleMul(
    B=&RigVMFunction_MathDoubleRemap(
     SourceMaximum=29
     SourceMinimum=58.69
     TargetMaximum=1
     TargetMinimum=0
     Value=&RigVMFunction_MathVectorDistance(B=@n12.Transform.Translation, A=@n13.Transform.Translation).Result
    ).Result
    A=@n11.OutputParam
   )
   > &RigUnit_ModifyTransforms(
    ItemToModify="((Item=(Type=Bone,Name=\"orient_coat_front_lf\"),Transform=(Rotation=(X=0,Y=0,Z=-0.0436,W=0.999),Translation=(X=0,Y=-1.9533,Z=0),Scale3D=(X=1,Y=1,Z=1))),(Item=(Type=Bone,Name=\"offset_coat_front_lf\"),Transform=(Rotation=(X=0,Y=0,Z=-0.887,W=0.4617),Translation ...(truncated)"
    WeightMinimum=0
    Weight=@n14.Result
   )
   n15 = &RigVMFunction_AlphaInterp(
    InterpSpeedDecreasing=5
    InterpSpeedIncreasing=100
    Scale=1
    bClampResult=true
    Value=&RigVMFunction_MathBoolToFloat(Value=RigVMVariableNode() {variable=IsCrouchWalking}.Value).Result
   )
   > &RigUnit_ModifyTransforms(
    ItemToModify="((Item=(Type=Bone,Name=\"orient_coat_front_lf\"),Transform=(Rotation=(X=0,Y=0.2164,Z=0,W=0.9763),Translation=(X=0,Y=0,Z=0),Scale3D=(X=1,Y=1,Z=1))),(Item=(Type=Bone,Name=\"offset_coat_front_lf\"),Transform=(Rotation=(X=0,Y=0.0872,Z=0,W=-0.9962),Translation=(X=0 ...(truncated)"
    WeightMinimum=0
    Weight=&RigVMFunction_MathDoubleMul(A=@n15.Result, B=@n14.Result).Result
   )
   RigVMFunction_MathBoolNot = &RigVMFunction_MathBoolNot(Value=RigVMVariableNode() {variable=IsRail}.Value)
   n16 = &RigVMFunction_AlphaInterp(
    InterpSpeedDecreasing=10
    InterpSpeedIncreasing=10
    Scale=1
    bClampResult=true
    bMapRange=true
    Value=&RigVMFunction_MathBoolToFloat(Value=@RigVMFunction_MathBoolNot.Result).Result
   )
   > &RigUnit_RotationConstraintLocalSpaceOffset(
    AdvancedSettings=(InterpolationType=Average,RotationOrderForFilter=XZY)
    Child=(Type=Bone,Name="coat_cr")
    Filter=(bX=True,bY=false,bZ=false)
    Parents=((Item=(Type=Bone,Name="pelvis"),Weight=1))
    Weight=@n16.Result
   )
   > n17 = &RigUnit_SphericalPoseReader(
    ActiveRegionScaleFactors=(PositiveWidth=1,NegativeWidth=1,PositiveHeight=1,NegativeHeight=1)
    ActiveRegionSize=0.05
    Debug=(bDrawDebug=false,bDraw2D=False,bDrawLocalAxes=False,DebugScale=25,DebugSegments=20,DebugThickness=0.25)
    DriverAxis=(X=-1,Y=0,Z=0)
    DriverItem=(Type=Bone,Name="thigh_l")
    FalloffRegionScaleFactors=(PositiveWidth=0.75,NegativeWidth=0.75,PositiveHeight=0.9,NegativeHeight=0.9)
    FalloffSize=0.63
    FlipHeightScaling=False
    FlipWidthScaling=False
    OptionalParentItem=(Type=Bone,Name="None")
    RotationOffset=(X=-90,Y=0,Z=10)
   )
   > &RigUnit_ModifyTransforms(
    ItemToModify="((Item=(Type=Bone,Name=\"base_coat_back_side_lf\"),Transform=(Rotation=(X=0,Y=0,Z=-0.6088,W=0.7934),Translation=(X=0,Y=0,Z=0),Scale3D=(X=1,Y=1,Z=1))),(Item=(Type=Bone,Name=\"base_coat_back_lf\"),Transform=(Rotation=(X=0,Y=0,Z=-0.7373,W=0.6756),Translation=(X=0 ...(truncated)"
    WeightMinimum=0
    Weight=@n17.OutputParam
   )
   > n18 = &RigUnit_SphericalPoseReader(
    ActiveRegionScaleFactors=(PositiveWidth=1,NegativeWidth=1,PositiveHeight=1,NegativeHeight=1)
    ActiveRegionSize=0.05
    Debug=(bDrawDebug=false,bDraw2D=False,bDrawLocalAxes=False,DebugScale=25,DebugSegments=20,DebugThickness=0.25)
    DriverAxis=(X=-1,Y=0,Z=0)
    DriverItem=(Type=Bone,Name="thigh_l")
    FalloffRegionScaleFactors=(PositiveWidth=1,NegativeWidth=0.75,PositiveHeight=0.9,NegativeHeight=0.9)
    FalloffSize=0.63
    FlipHeightScaling=False
    FlipWidthScaling=false
    OptionalParentItem=(Type=Bone,Name="None")
    RotationOffset=(X=-135,Y=0,Z=0)
   )
   > &RigUnit_ModifyTransforms(
    ItemToModify="((Item=(Type=Bone,Name=\"base_coat_back_side_lf\"),Transform=(Rotation=(X=0.0057,Y=-0.0432,Z=-0.4613,W=0.8862),Translation=(X=0,Y=0,Z=0),Scale3D=(X=1,Y=1,Z=1))),(Item=(Type=Bone,Name=\"base_coat_back_lf\"),Transform=(Rotation=(X=0,Y=0,Z=0,W=1),Translation=(X=0 ...(truncated)"
    WeightMinimum=0
    Weight=@n18.OutputParam
   )
   n19 = &RigUnit_SphericalPoseReader(
    ActiveRegionScaleFactors=(PositiveWidth=1,NegativeWidth=1,PositiveHeight=1,NegativeHeight=1)
    ActiveRegionSize=0.4
    Debug=(bDrawDebug=false,bDraw2D=False,bDrawLocalAxes=False,DebugScale=25,DebugSegments=20,DebugThickness=0.25)
    DriverAxis=(X=1,Y=0,Z=0)
    DriverItem=(Type=Bone,Name="thigh_r")
    FalloffRegionScaleFactors=(PositiveWidth=1,NegativeWidth=1,PositiveHeight=1,NegativeHeight=1)
    FalloffSize=0.6
    FlipHeightScaling=False
    FlipWidthScaling=False
    OptionalParentItem=(Type=Bone,Name="None")
    RotationOffset=(X=-90,Y=0,Z=-20)
    ExecutePin=@RigVMFunction_Sequence_1.B
   )
   n20 = &RigUnit_GetTransform(Item=(Type=Bone,Name="calf_r"), bInitial=False)
   n21 = &RigUnit_GetTransform(Item=(Type=Bone,Name="tgt_coat_front_rt"), bInitial=False)
   n22 = &RigVMFunction_MathDoubleMul(
    B=&RigVMFunction_MathDoubleRemap(
     SourceMaximum=11.6097
     SourceMinimum=60.629
     TargetMaximum=1
     TargetMinimum=0
     Value=&RigVMFunction_MathVectorDistance(B=@n20.Transform.Translation, A=@n21.Transform.Translation).Result
    ).Result
    A=@n19.OutputParam
   )
   > &RigUnit_ModifyTransforms(
    ItemToModify="((Item=(Type=Bone,Name=\"orient_coat_front_rt\"),Transform=(Rotation=(X=-0.4126,Y=0.056,Z=-0.2091,W=0.8848),Translation=(X=0,Y=-2.6669,Z=7.2002),Scale3D=(X=1,Y=1,Z=1))),(Item=(Type=Bone,Name=\"offset_coat_front_rt\"),Transform=(Rotation=(X=-0.4034,Y=-0.2177,Z= ...(truncated)"
    WeightMinimum=0
    Weight=@n22.Result
   )
   n23 = &RigVMFunction_AlphaInterp(
    InterpSpeedDecreasing=5
    InterpSpeedIncreasing=100
    Scale=1
    bClampResult=true
    Value=&RigVMFunction_MathBoolToFloat(Value=RigVMVariableNode() {variable=IsCrouchWalking}.Value).Result
   )
   > &RigUnit_ModifyTransforms(
    ItemToModify="((Item=(Type=Bone,Name=\"orient_coat_front_rt\"),Transform=(Rotation=(X=-0.1305,Y=0,Z=0,W=0.9914),Translation=(X=0,Y=0,Z=0.8655),Scale3D=(X=1,Y=1,Z=1))),(Item=(Type=Bone,Name=\"offset_coat_front_rt\"),Transform=(Rotation=(X=0,Y=-0.2164,Z=0,W=0.9763),Translatio ...(truncated)"
    WeightMinimum=0
    Weight=&RigVMFunction_MathDoubleMul(A=@n23.Result, B=@n22.Result).Result
   )
   > n24 = &RigUnit_SphericalPoseReader(
    ActiveRegionScaleFactors=(PositiveWidth=0.5,NegativeWidth=1,PositiveHeight=0.5,NegativeHeight=0.5)
    ActiveRegionSize=0.35
    Debug=(bDrawDebug=false,bDraw2D=False,bDrawLocalAxes=False,DebugScale=25,DebugSegments=20,DebugThickness=0.25)
    DriverAxis=(X=1,Y=0,Z=0)
    DriverItem=(Type=Bone,Name="thigh_r")
    FalloffRegionScaleFactors=(PositiveWidth=0.5,NegativeWidth=1,PositiveHeight=0.5,NegativeHeight=0.5)
    FalloffSize=0.55
    FlipHeightScaling=False
    FlipWidthScaling=False
    OptionalParentItem=(Type=Bone,Name="None")
    RotationOffset=(X=0,Y=-20,Z=180)
   )
   n25 = &RigUnit_GetTransform(Item=(Type=Bone,Name="calf_r"), bInitial=False)
   n26 = &RigUnit_GetTransform(Item=(Type=Bone,Name="tgt_coat_side_rt"), bInitial=False)
   n27 = &RigVMFunction_MathDoubleMul(
    B=&RigVMFunction_MathDoubleRemap(
     SourceMaximum=22
     SourceMinimum=58.69
     TargetMaximum=1
     TargetMinimum=0
     Value=&RigVMFunction_MathVectorDistance(B=@n25.Transform.Translation, A=@n26.Transform.Translation).Result
    ).Result
    A=@n24.OutputParam
   )
   > &RigUnit_ModifyTransforms(
    ItemToModify="((Item=(Type=Bone,Name=\"offset_coat_back_side_rt\"),Transform=(Rotation=(X=0.1367,Y=-0.4166,Z=0.1701,W=-0.8825),Translation=(X=0,Y=0,Z=0),Scale3D=(X=1,Y=1,Z=1))),(Item=(Type=Bone,Name=\"offset_coat_front_rt\"),Transform=(Rotation=(X=0,Y=0.1305,Z=0,W=-0.9914), ...(truncated)"
    WeightMinimum=0
    Weight=@n27.Result
   )
   n28 = &RigVMFunction_AlphaInterp(
    InterpSpeedDecreasing=5
    InterpSpeedIncreasing=100
    Scale=1
    bClampResult=true
    Value=&RigVMFunction_MathBoolToFloat(Value=RigVMVariableNode() {variable=IsCrouchWalking}.Value).Result
   )
   > &RigUnit_ModifyTransforms(
    ItemToModify="((Item=(Type=Bone,Name=\"offset_coat_back_side_rt\"),Transform=(Rotation=(X=0,Y=0,Z=0,W=1),Translation=(X=0,Y=0,Z=0),Scale3D=(X=1,Y=1,Z=1))),(Item=(Type=Bone,Name=\"offset_coat_front_rt\"),Transform=(Rotation=(X=0,Y=0,Z=0,W=1),Translation=(X=0,Y=0,Z=0),Scale3D ...(truncated)"
    WeightMinimum=0
    Weight=&RigVMFunction_MathDoubleMul(A=@n28.Result, B=@n27.Result).Result
   )
   > n29 = &RigUnit_SphericalPoseReader(
    ActiveRegionScaleFactors=(PositiveWidth=1,NegativeWidth=1,PositiveHeight=0.5,NegativeHeight=0.5)
    ActiveRegionSize=0.35
    Debug=(bDrawDebug=false,bDraw2D=False,bDrawLocalAxes=False,DebugScale=25,DebugSegments=20,DebugThickness=0.25)
    DriverAxis=(X=1,Y=0,Z=0)
    DriverItem=(Type=Bone,Name="thigh_r")
    FalloffRegionScaleFactors=(PositiveWidth=0.5,NegativeWidth=1,PositiveHeight=1,NegativeHeight=0.7)
    FalloffSize=0.58
    FlipHeightScaling=False
    FlipWidthScaling=true
    OptionalParentItem=(Type=Bone,Name="None")
    RotationOffset=(X=-30,Y=20,Z=0)
   )
   n30 = &RigUnit_GetTransform(Item=(Type=Bone,Name="calf_r"), bInitial=False)
   n31 = &RigUnit_GetTransform(Item=(Type=Bone,Name="tgt_coat_front_side_rt"), bInitial=False)
   n32 = &RigVMFunction_MathDoubleMul(
    B=&RigVMFunction_MathDoubleRemap(
     SourceMaximum=29
     SourceMinimum=58.69
     TargetMaximum=1
     TargetMinimum=0
     Value=&RigVMFunction_MathVectorDistance(B=@n30.Transform.Translation, A=@n31.Transform.Translation).Result
    ).Result
    A=@n29.OutputParam
   )
   > &RigUnit_ModifyTransforms(
    ItemToModify="((Item=(Type=Bone,Name=\"orient_coat_front_rt\"),Transform=(Rotation=(X=0,Y=0,Z=0,W=1),Translation=(X=0,Y=0,Z=0),Scale3D=(X=1,Y=1,Z=1))),(Item=(Type=Bone,Name=\"offset_coat_front_rt\"),Transform=(Rotation=(X=0,Y=0,Z=-0.7373,W=0.6756),Translation=(X=0,Y=0,Z=0), ...(truncated)"
    WeightMinimum=0
    Weight=@n32.Result
   )
   n33 = &RigVMFunction_AlphaInterp(
    InterpSpeedDecreasing=5
    InterpSpeedIncreasing=100
    Scale=1
    bClampResult=true
    Value=&RigVMFunction_MathBoolToFloat(Value=RigVMVariableNode() {variable=IsCrouchWalking}.Value).Result
   )
   > &RigUnit_ModifyTransforms(
    ItemToModify="((Item=(Type=Bone,Name=\"orient_coat_front_rt\"),Transform=(Rotation=(X=-0.3827,Y=0,Z=0,W=-0.9239),Translation=(X=0,Y=-2.4104,Z=-2.4104),Scale3D=(X=1,Y=1,Z=1))),(Item=(Type=Bone,Name=\"offset_coat_front_rt\"),Transform=(Rotation=(X=-0.0872,Y=0,Z=0,W=-0.9962),T ...(truncated)"
    WeightMinimum=0
    Weight=&RigVMFunction_MathDoubleMul(A=@n33.Result, B=@n32.Result).Result
   )
   > n34 = &RigUnit_SphericalPoseReader(
    ActiveRegionScaleFactors=(PositiveWidth=1,NegativeWidth=1,PositiveHeight=1,NegativeHeight=1)
    ActiveRegionSize=0.05
    Debug=(bDrawDebug=false,bDraw2D=False,bDrawLocalAxes=False,DebugScale=25,DebugSegments=20,DebugThickness=0.25)
    DriverAxis=(X=1,Y=0,Z=0)
    DriverItem=(Type=Bone,Name="thigh_r")
    FalloffRegionScaleFactors=(PositiveWidth=0.75,NegativeWidth=0.75,PositiveHeight=0.9,NegativeHeight=0.9)
    FalloffSize=0.63
    FlipHeightScaling=False
    FlipWidthScaling=False
    OptionalParentItem=(Type=Bone,Name="None")
    RotationOffset=(X=90,Y=0,Z=10)
   )
   > &RigUnit_ModifyTransforms(
    ItemToModify="((Item=(Type=Bone,Name=\"base_coat_back_side_rt\"),Transform=(Rotation=(X=0,Y=0,Z=-0.766,W=0.6428),Translation=(X=0,Y=0,Z=0),Scale3D=(X=1,Y=1,Z=1))),(Item=(Type=Bone,Name=\"base_coat_back_rt\"),Transform=(Rotation=(X=0,Y=0,Z=-0.9397,W=0.342),Translation=(X=0,Y ...(truncated)"
    WeightMinimum=0
    Weight=@n34.OutputParam
   )
   > n35 = &RigUnit_SphericalPoseReader(
    ActiveRegionScaleFactors=(PositiveWidth=1,NegativeWidth=1,PositiveHeight=1,NegativeHeight=1)
    ActiveRegionSize=0.05
    Debug=(bDrawDebug=false,bDraw2D=False,bDrawLocalAxes=False,DebugScale=25,DebugSegments=20,DebugThickness=0.25)
    DriverAxis=(X=1,Y=0,Z=0)
    DriverItem=(Type=Bone,Name="thigh_r")
    FalloffRegionScaleFactors=(PositiveWidth=0.75,NegativeWidth=1,PositiveHeight=0.9,NegativeHeight=0.9)
    FalloffSize=0.63
    FlipHeightScaling=False
    FlipWidthScaling=false
    OptionalParentItem=(Type=Bone,Name="None")
    RotationOffset=(X=45,Y=0,Z=0)
   )
   > &RigUnit_ModifyTransforms(
    ItemToModify="((Item=(Type=Bone,Name=\"base_coat_back_side_rt\"),Transform=(Rotation=(X=0,Y=0,Z=-0.0872,W=0.9962),Translation=(X=0,Y=0,Z=0),Scale3D=(X=1,Y=1,Z=1))),(Item=(Type=Bone,Name=\"base_coat_back_rt\"),Transform=(Rotation=(X=0,Y=0,Z=-0.0872,W=0.9962),Translation=(X=0 ...(truncated)"
    WeightMinimum=0
    Weight=@n35.OutputParam
   )
   > &RigUnit_AimConstraintLocalSpaceOffset(
    Child=(Type=Bone,Name="base_jacket_spine_cr")
    Filter=(bX=True,bY=True,bZ=True)
    Parents=((Item=(Type=Bone,Name="pelvis"),Weight=1))
    UpAxis=(X=0,Y=1,Z=0)
    Weight=1
    WorldUp=(Target=(X=0,Y=1,Z=0),Kind=Direction,Space=(Type=Bone,Name="base_jacket_spine_cr"))
   )
   n36 = &RigVMFunction_AlphaInterp(
    InterpSpeedDecreasing=10
    InterpSpeedIncreasing=10
    Scale=1
    bClampResult=true
    bMapRange=true
    Value=&RigVMFunction_MathBoolToFloat(Value=RigVMVariableNode() {variable=IsMeleeNotCrouch}.Value).Result
   )
   > &RigUnit_ModifyTransforms(
    ItemToModify="((Item=(Type=Bone,Name=\"offset_coat_front_lf\"),Transform=(Rotation=(X=0,Y=0,Z=0,W=1),Translation=(X=0,Y=-3,Z=0),Scale3D=(X=1,Y=1,Z=1))),(Item=(Type=Bone,Name=\"offset_coat_front_side_lf\"),Transform=(Rotation=(X=0,Y=0,Z=0,W=1),Translation=(X=0,Y=-3,Z=0),Scal ...(truncated)"
    WeightMinimum=0
    Weight=@n36.Result
   )
   n37 = &RigVMFunction_AlphaInterp(
    InterpSpeedDecreasing=10
    InterpSpeedIncreasing=10
    Scale=1
    bClampResult=true
    bMapRange=true
    Value=&RigVMFunction_MathBoolToFloat(Value=RigVMVariableNode() {variable=IsCrouch}.Value).Result
   )
   n38 = &RigVMFunction_MathDoubleMul(
    B=@n37.Result
    A=&RigVMFunction_MathDoubleRemap(
     SourceMaximum=45
     SourceMinimum=89
     TargetMaximum=1
     TargetMinimum=0
     Value=&RigUnit_GetRelativeTransformForItem(Child=(Type=Bone,Name="pelvis"), Parent=(Type=Bone,Name="root"), bChildInitial=False, bParentInitial=False).RelativeTransform.Translation.Z
    ).Result
   )
   &RigUnit_ModifyTransforms(
    ItemToModify="((Item=(Type=Bone,Name=\"orient_coat_front_lf\"),Transform=(Rotation=(X=0,Y=0.2164,Z=0,W=0.9763),Translation=(X=-2.1549,Y=0,Z=2.0935),Scale3D=(X=1,Y=1,Z=1))),(Item=(Type=Bone,Name=\"offset_coat_front_lf\"),Transform=(Rotation=(X=-0.0019,Y=-0.0436,Z=-0.0436,W=0 ...(truncated)"
    WeightMinimum=0
    Weight=@n38.Result
    ExecutePin=@RigVMFunction_Sequence_1.C
   )
   > &RigUnit_ModifyTransforms(
    ItemToModify="((Item=(Type=Bone,Name=\"offset_coat_back_side_lf\"),Transform=(Rotation=(X=0,Y=0,Z=-0.0872,W=-0.9962),Translation=(X=0,Y=0,Z=0),Scale3D=(X=1,Y=1,Z=1))),(Item=(Type=Bone,Name=\"offset_coat_back_lf\"),Transform=(Rotation=(X=0,Y=0,Z=0.0872,W=0.9962),Translation= ...(truncated)"
    WeightMinimum=0
    Weight=@n38.Result
   )
   n39 = RigVMVariableNode() {variable=IsCrouchWalking}
   n40 = RigVMVariableNode() {variable=isMelee}
   n41 = &RigVMFunction_AlphaInterp(
    InterpSpeedDecreasing=5
    InterpSpeedIncreasing=100
    Scale=1
    bClampResult=true
    Value=&RigVMFunction_MathBoolToFloat(Value=&RigVMFunction_MathBoolAnd(A=@n39.Value, B=@n40.Value).Result).Result
   )
   > &RigUnit_ModifyTransforms(
    ItemToModify="((Item=(Type=Bone,Name=\"offset_coat_front_side_rt\"),Transform=(Rotation=(X=0,Y=0,Z=0.1305,W=-0.9914),Translation=(X=6.1572,Y=0.4869,Z=-0.6549),Scale3D=(X=1,Y=1,Z=1))),(Item=(Type=Bone,Name=\"offset_coat_side_rt\"),Transform=(Rotation=(X=0,Y=0,Z=0,W=-1),Trans ...(truncated)"
    WeightMinimum=0
    Weight=@n41.Result
   )
   n42 = &RigVMFunction_AlphaInterp(
    InterpSpeedDecreasing=5
    InterpSpeedIncreasing=100
    Scale=1
    bClampResult=true
    Value=&RigVMFunction_MathBoolToFloat(Value=RigVMVariableNode() {variable=IsCrouchWalking}.Value).Result
   )
   > &RigUnit_ModifyTransforms(
    ItemToModify="((Item=(Type=Bone,Name=\"orient_coat_front_lf\"),Transform=(Rotation=(X=0,Y=0,Z=0,W=1),Translation=(X=0,Y=-5,Z=0),Scale3D=(X=1,Y=1,Z=1))),(Item=(Type=Bone,Name=\"offset_coat_front_lf\"),Transform=(Rotation=(X=0,Y=0,Z=0,W=1),Translation=(X=0,Y=0,Z=0),Scale3D=(X ...(truncated)"
    WeightMinimum=0
    Weight=@n42.Result
   )
   n43 = RigVMVariableNode() {variable=IsOnAir}
   &RigUnit_AimConstraintLocalSpaceOffset(
    Child=(Type=Bone,Name="pelvis_cloth_helper")
    Filter=(bX=false,bY=True,bZ=True)
    Parents=((Item=(Type=Bone,Name="root"),Weight=1))
    WorldUp=(Target=(X=0,Y=0,Z=1),Kind=Direction,Space=(Type=Bone,Name="root"))
    ExecutePin=@n4.True
    Weight=&RigVMFunction_MathBoolToFloat(Value=&RigVMFunction_MathBoolNot(Value=@n43.Value).Result).Result
   )
   n44 = &RigVMFunction_AlphaInterp(
    InterpSpeedDecreasing=5
    InterpSpeedIncreasing=100
    Scale=1
    bClampResult=true
    Value=&RigVMFunction_MathBoolToFloat(Value=RigVMVariableNode() {variable=IsCrouchWalking}.Value).Result
   )
   &RigUnit_ModifyTransforms(
    ExecutePin=()
    ItemToModify="((Item=(Type=Bone,Name=\"orient_coat_front_rt\"),Transform=(Rotation=(X=0,Y=0,Z=0,W=1),Translation=(X=5,Y=0,Z=0),Scale3D=(X=1,Y=1,Z=1))),(Item=(Type=Bone,Name=\"orient_coat_front_lf\"),Transform=(Rotation=(X=0,Y=0,Z=0,W=1),Translation=(X=0,Y=-5,Z=-5),Scale3D=( ...(truncated)"
    WeightMinimum=0
    Weight=@n44.Result
   )
  ]}
]
hierarchy [
 root {loc=<0 0 0>, rot=<0 0 0 1>}
 pelvis {parent=root, loc=<0 -0.7071 93.1191>, rot=<0.0127 0.707 -0.0127 -0.707>}
 spine_01 {parent=pelvis, loc=<6.7494 -0.6388 0>, rot=<0 0 0.034 -0.9994>}
 spine_02 {parent=spine_01, loc=<7.2518 -0.035 0>, rot=<0 0 0.0238 -0.9997>}
 spine_03 {parent=spine_02, loc=<6.0104 0 0>, rot=<0 0 0.0149 -0.9999>}
 spine_04 {parent=spine_03, loc=<8.1024 0 0>, rot=<0 0 -0.0132 -0.9999>}
 spine_05 {parent=spine_04, loc=<10.0146 0 0>, rot=<0 0 -0.0015 -1>}
 neck_01 {parent=spine_05, loc=<11.3138 0.1529 0>, rot=<0 0 -0.2003 -0.9797>}
 neck_02 {parent=neck_01, loc=<4.5627 0 0>, rot=<0 0 0.0504 -0.9987>}
 head {parent=neck_02, loc=<5.3914 0 0>, rot=<0 0 0.115 -0.9934>}
 attach_cape {parent=spine_05, loc=<9.0156 -11.7124 -0.0028>, rot=<0.0616 -0.7044 -0.0616 -0.7044>}
 cape {parent=attach_cape, type=Null, loc=<0 0 0>, rot=<0 0 0 1>}
 attach_backpack {parent=spine_05, loc=<-0.1209 -13.4703 -0.0023>, rot=<-0.0093 -0.707 0.0093 -0.707>}
 Backpack_BR {parent=attach_backpack, type=Null, loc=<0 0 0>, rot=<0 0 0 1>}
 clavicle_l {parent=spine_05, loc=<4.3575 -0.1466 -2.3223>, rot=<0.0465 -0.7289 -0.009 -0.6829>}
 upperarm_l {parent=clavicle_l, loc=<15.5111 0 0>, rot=<0.0226 -0.3234 -0.0088 -0.9459>}
 lowerarm_l {parent=upperarm_l, loc=<26.6514 0 0>, rot=<0.0391 -0.0099 -0.2453 -0.9686>}
 hand_l {parent=lowerarm_l, loc=<22.3491 0 0>, rot=<-0.7081 -0.0507 0.0975 -0.6975>}
 thumb_01_l {parent=hand_l, loc=<1.3118 -0.3191 -2.459>, rot=<0.6546 -0.4042 0.0664 -0.6354>}
 thumb_02_l {parent=thumb_01_l, loc=<5.9336 -0.0448 0.0401>, rot=<-0.0037 0.0213 0.298 -0.9543>}
 thumb_03_l {parent=thumb_02_l, loc=<2.2223 0.3682 -0.0907>, rot=<0.0118 -0.013 0.0066 -0.9998>}
 index_metacarpal_l {parent=hand_l, loc=<1.6535 0.0183 -2.96>, rot=<0.0158 -0.0124 -0.0173 -0.9996>}
 index_01_l {parent=index_metacarpal_l, loc=<8.3008 0.0357 -0.0524>, rot=<-0.0044 0.0485 0.0924 -0.9945>}
 index_02_l {parent=index_01_l, loc=<3.3587 -0.3004 0.0452>, rot=<-0.002 0.0027 0.3225 -0.9466>}
 index_03_l {parent=index_02_l, loc=<2.5578 -0.123 -0.0252>, rot=<-0.002 0.0067 0.2558 -0.9667>}
 middle_metacarpal_l {parent=hand_l, loc=<2.1856 -0.1725 -0.5984>, rot=<0.006 0.0168 -0.0329 -0.9993>}
 middle_01_l {parent=middle_metacarpal_l, loc=<7.3275 -0.2447 0.0129>, rot=<-0.0208 0.0042 0.0858 -0.9961>}
 middle_02_l {parent=middle_01_l, loc=<4.0123 0.1401 0.1175>, rot=<-0.0218 0.0258 0.3512 -0.9357>}
 middle_03_l {parent=middle_02_l, loc=<2.6077 0.0534 0.1373>, rot=<0.0082 0.0188 0.1708 -0.9851>}
 ring_metacarpal_l {parent=hand_l, loc=<2.2074 -0.3625 1.3638>, rot=<-0.0284 0.0722 -0.0458 -0.9959>}
 ring_01_l {parent=ring_metacarpal_l, loc=<7.0295 -0.6197 -0.1446>, rot=<-0.0489 -0.0252 0.1892 -0.9804>}
 ring_02_l {parent=ring_01_l, loc=<3.6174 0.4647 -0.1027>, rot=<0 0 0.2806 -0.9598>}
 ring_03_l {parent=ring_02_l, loc=<2.4152 0.114 0.0199>, rot=<0 0 0.1189 -0.9929>}
 pinky_metacarpal_l {parent=hand_l, loc=<1.7616 -0.9753 3.3814>, rot=<-0.0397 0.0814 -0.069 -0.9935>}
 pinky_01_l {parent=pinky_metacarpal_l, loc=<7.1709 -0.4831 -0.1094>, rot=<-0.0654 -0.0035 0.2242 -0.9723>}
 pinky_02_l {parent=pinky_01_l, loc=<2.8274 0.0893 -0.1043>, rot=<0.0037 0.0011 0.2854 -0.9584>}
 pinky_03_l {parent=pinky_02_l, loc=<2.2046 0.0307 -0.0078>, rot=<-0.0002 0.0098 0.1445 -0.9895>}
 deform_wrist_dn_base_l {parent=hand_l, loc=<-3.3507 -3.6133 -0.9135>, rot=<0.0126 0.0008 -0.0701 -0.9975>}
 deform_wrist_dn_l {parent=deform_wrist_dn_base_l, loc=<0 -0.1804 0>, rot=<0 0 0 1>, scale=<0.985 1 1>}
 deform_wrist_up_base_l {parent=hand_l, loc=<-3.887 0.1813 -1.0101>, rot=<0.0126 0.0008 -0.0701 -0.9975>}
 deform_wrist_up_l {parent=deform_wrist_up_base_l, loc=<0 0 0>, rot=<0 0 0 1>}
 lowerarm_twist_02_l {parent=lowerarm_l, loc=<9.6199 0 0>, rot=<0 0 0 -1>}
 lowerarm_twist_01_l {parent=lowerarm_l, loc=<17.8549 0 0>, rot=<0 0 0 -1>}
 deform_elbow_in_base_l {parent=lowerarm_l, loc=<-4.0521 5.0615 -2.2346>, rot=<-0.0285 0.0335 -0.0024 -0.999>}
 deform_elbow_in_l {parent=deform_elbow_in_base_l, loc=<0 0 0>, rot=<0 0 0 1>}
 deform_elbow_out_base_l {parent=lowerarm_l, loc=<-4.2019 -2.5475 -2.4231>, rot=<-0.0285 0.0335 -0.0024 -0.999>}
 deform_elbow_out_l {parent=deform_elbow_out_base_l, loc=<0 0 0>, rot=<0 0 0 1>}
 upperarm_twist_01_l {parent=upperarm_l, loc=<0.0041 0 0>, rot=<0 0 0 -1>}
 deform_upperarm_bk_base_l {parent=upperarm_twist_01_l, loc=<-3.2582 -5.546 -2.0366>, rot=<0 -0.0007 -0.0191 -0.9998>}
 deform_upperarm_bk_l {parent=deform_upperarm_bk_base_l, loc=<0 0 0>, rot=<0 0 0 1>}
 deform_upperarm_fr_base_l {parent=upperarm_twist_01_l, loc=<-2.8462 3.9433 -2.0158>, rot=<0 -0.0007 -0.0191 -0.9998>}
 deform_upperarm_fr_l {parent=deform_upperarm_fr_base_l, loc=<0 0 0>, rot=<0 0 0 1>}
 deform_upperarm_up_base_l {parent=upperarm_twist_01_l, loc=<-5.9223 -0.7874 1.605>, rot=<0 -0.0007 -0.0191 -0.9998>}
 deform_upperarm_up_l {parent=deform_upperarm_up_base_l, loc=<0 0 0>, rot=<0 0 0 1>}
 upperarm_twist_02_l {parent=upperarm_l, loc=<13.7941 0 0>, rot=<0 0 0 -1>}
 clavicle_r {parent=spine_05, loc=<4.3575 -0.1466 2.3223>, rot=<0.7289 0.0465 -0.6829 0.009>}
 upperarm_r {parent=clavicle_r, loc=<-15.5111 0 0>, rot=<0.0226 -0.3234 -0.0088 -0.9459>}
 lowerarm_r {parent=upperarm_r, loc=<-26.6514 0 0>, rot=<0.0391 -0.0099 -0.2453 -0.9686>}
 hand_r {parent=lowerarm_r, loc=<-22.3491 0 0>, rot=<-0.7081 -0.0507 0.0975 -0.6975>}
 thumb_01_r {parent=hand_r, loc=<-1.3118 0.3191 2.459>, rot=<0.6546 -0.4042 0.0664 -0.6354>}
 thumb_02_r {parent=thumb_01_r, loc=<-5.9336 0.0448 -0.0401>, rot=<-0.0037 0.0213 0.298 -0.9543>}
 thumb_03_r {parent=thumb_02_r, loc=<-2.2223 -0.3682 0.0907>, rot=<0.0118 -0.013 0.0066 -0.9998>}
 index_metacarpal_r {parent=hand_r, loc=<-1.6535 -0.0183 2.96>, rot=<0.0158 -0.0124 -0.0173 -0.9996>}
 index_01_r {parent=index_metacarpal_r, loc=<-8.3008 -0.0357 0.0524>, rot=<-0.0044 0.0485 0.0924 -0.9945>}
 index_02_r {parent=index_01_r, loc=<-3.3587 0.3004 -0.0452>, rot=<-0.002 0.0027 0.3225 -0.9466>}
 index_03_r {parent=index_02_r, loc=<-2.5578 0.123 0.0252>, rot=<-0.002 0.0067 0.2558 -0.9667>}
 middle_metacarpal_r {parent=hand_r, loc=<-2.1856 0.1725 0.5984>, rot=<0.006 0.0168 -0.0329 -0.9993>}
 middle_01_r {parent=middle_metacarpal_r, loc=<-7.3275 0.2447 -0.0129>, rot=<-0.0208 0.0042 0.0858 -0.9961>}
 middle_02_r {parent=middle_01_r, loc=<-4.0123 -0.1401 -0.1175>, rot=<-0.0218 0.0258 0.3512 -0.9357>}
 middle_03_r {parent=middle_02_r, loc=<-2.6077 -0.0534 -0.1373>, rot=<0.0082 0.0188 0.1708 -0.9851>}
 ring_metacarpal_r {parent=hand_r, loc=<-2.2074 0.3625 -1.3638>, rot=<-0.0284 0.0722 -0.0458 -0.9959>}
 ring_01_r {parent=ring_metacarpal_r, loc=<-7.0295 0.6197 0.1446>, rot=<-0.0489 -0.0252 0.1892 -0.9804>}
 ring_02_r {parent=ring_01_r, loc=<-3.6174 -0.4647 0.1027>, rot=<0 0 0.2806 -0.9598>}
 ring_03_r {parent=ring_02_r, loc=<-2.4152 -0.114 -0.0199>, rot=<0 0 0.1189 -0.9929>}
 pinky_metacarpal_r {parent=hand_r, loc=<-1.7616 0.9753 -3.3814>, rot=<-0.0397 0.0814 -0.069 -0.9935>}
 pinky_01_r {parent=pinky_metacarpal_r, loc=<-7.1709 0.4831 0.1094>, rot=<-0.0654 -0.0035 0.2242 -0.9723>}
 pinky_02_r {parent=pinky_01_r, loc=<-2.8274 -0.0893 0.1043>, rot=<0.0037 0.0011 0.2854 -0.9584>}
 pinky_03_r {parent=pinky_02_r, loc=<-2.2046 -0.0307 0.0078>, rot=<-0.0002 0.0098 0.1445 -0.9895>}
 deform_wrist_dn_base_r {parent=hand_r, loc=<3.3505 3.6135 0.9135>, rot=<0.0126 0.0008 -0.0701 -0.9975>}
 deform_wrist_dn_r {parent=deform_wrist_dn_base_r, loc=<0 0.1804 0>, rot=<0 0 0 1>, scale=<0.985 1 1>}
 deform_wrist_up_base_r {parent=hand_r, loc=<3.8867 -0.1811 1.01>, rot=<0.0126 0.0008 -0.0701 -0.9975>}
 deform_wrist_up_r {parent=deform_wrist_up_base_r, loc=<0 0 0>, rot=<0 0 0 1>}
 lowerarm_twist_02_r {parent=lowerarm_r, loc=<-9.6199 0 0>, rot=<0 0 0 -1>}
 lowerarm_twist_01_r {parent=lowerarm_r, loc=<-17.8549 0 0>, rot=<0 0 0 -1>}
 deform_elbow_in_base_r {parent=lowerarm_r, loc=<4.0521 -5.0615 2.2346>, rot=<-0.0285 0.0335 -0.0024 -0.999>}
 deform_elbow_in_r {parent=deform_elbow_in_base_r, loc=<0 0 0>, rot=<0 0 0 1>}
 deform_elbow_out_base_r {parent=lowerarm_r, loc=<4.2019 2.5475 2.423>, rot=<-0.0285 0.0335 -0.0024 -0.999>}
 deform_elbow_out_r {parent=deform_elbow_out_base_r, loc=<0 0 0>, rot=<0 0 0 1>}
 upperarm_twist_01_r {parent=upperarm_r, loc=<-0.0041 0 0>, rot=<0 0 0 -1>}
 deform_upperarm_bk_base_r {parent=upperarm_twist_01_r, loc=<3.2582 5.546 2.0366>, rot=<0 -0.0007 -0.0191 -0.9998>}
 deform_upperarm_bk_r {parent=deform_upperarm_bk_base_r, loc=<0 0 0>, rot=<0 0 0 1>}
 deform_upperarm_fr_base_r {parent=upperarm_twist_01_r, loc=<2.8463 -3.9433 2.0158>, rot=<0 -0.0007 -0.0191 -0.9998>}
 deform_upperarm_fr_r {parent=deform_upperarm_fr_base_r, loc=<0 0 0>, rot=<0 0 0 1>}
 deform_upperarm_up_base_r {parent=upperarm_twist_01_r, loc=<5.9223 0.7874 -1.605>, rot=<0 -0.0007 -0.0191 -0.9998>}
 deform_upperarm_up_r {parent=deform_upperarm_up_base_r, loc=<0 0 0>, rot=<0 0 0 1>}
 upperarm_twist_02_r {parent=upperarm_r, loc=<-13.7941 0 0>, rot=<0 0 0 -1>}
 base_hood_rope_front_01_rt {parent=spine_05, loc=<6.9529 10.6835 6.5826>, rot=<0.0405 -0.0513 0.9806 0.185>}
 dyn_hood_rope_front_01_rt {parent=base_hood_rope_front_01_rt, loc=<0 0 0>, rot=<0 0 0 1>}
 dyn_hood_rope_front_02_rt {parent=dyn_hood_rope_front_01_rt, loc=<4.6316 0.0585 0.0019>, rot=<-0.0097 -0.0169 -0.1127 -0.9934>}
 dyn_hood_rope_front_03_rt {parent=dyn_hood_rope_front_02_rt, loc=<4.6024 0.041 0.0076>, rot=<-0.0073 -0.0165 -0.0442 -0.9989>}
 dyn_hood_rope_front_04_rt {parent=dyn_hood_rope_front_03_rt, loc=<4.7229 -0.0336 0>, rot=<0.0004 -0.0258 -0.0154 -0.9995>}
 base_hood_rope_front_01_lf {parent=spine_05, loc=<7.0706 10.6924 -6.5966>, rot=<-0.0376 0.0636 0.9783 0.1937>}
 dyn_hood_rope_front_01_lf {parent=base_hood_rope_front_01_lf, loc=<0 0 0>, rot=<0 0 0 1>}
 dyn_hood_rope_front_02_lf {parent=dyn_hood_rope_front_01_lf, loc=<3.3829 0.0306 -0.0033>, rot=<0.0284 0.0135 -0.1087 -0.9936>}
 dyn_hood_rope_front_03_lf {parent=dyn_hood_rope_front_02_lf, loc=<4.2515 -0.0197 0>, rot=<-0.0016 0.0323 -0.0493 -0.9983>}
 base_hood_back_cr_lower {parent=spine_05, loc=<15.2598 -7.175 -0.1521>, rot=<0.1413 -0.9895 0.025 -0.0141>}
 dyn_hood_back_01_cr_lower {parent=base_hood_back_cr_lower, loc=<0 0 0>, rot=<0 0 0 1>}
 head_null {parent=spine_05, type=Null, loc=<20.6573 3.5489 0>, rot=<0 0 -0.036 -0.9994>}
 offset_hood_back_cr_top {parent=spine_05, loc=<11.3138 0.1529 0>, rot=<0.3428 -0.9394 0 0>}
 base_hood_back_cr_top {parent=offset_hood_back_cr_top, loc=<0.04 -9.1359 0.105>, rot=<-0.0283 -0.005 0.1315 -0.9909>}
 dyn_hood_back_01_cr_top {parent=base_hood_back_cr_top, loc=<0 0 0>, rot=<0 0 0 1>}
 deform_pec_aim_base_l {parent=spine_04, loc=<6.2615 9.3309 -1.2237>, rot=<0.0603 -0.4033 0.3318 -0.8507>}
 deform_pec_aim_l {parent=deform_pec_aim_base_l, loc=<0 0 0>, rot=<0 0 0 1>}
 deform_pec_base_l {parent=deform_pec_aim_l, loc=<9.3665 0 0>, rot=<0 0 0 1>}
 deform_pec_l {parent=deform_pec_base_l, loc=<0 0 0>, rot=<0 0 0 1>}
 deform_pec_aim_base_r {parent=spine_04, loc=<6.2611 9.3309 1.2237>, rot=<-0.4033 -0.0603 0.8507 0.3318>}
 deform_pec_aim_r {parent=deform_pec_aim_base_r, loc=<0 0 0>, rot=<0 0 0 1>}
 deform_pec_base_r {parent=deform_pec_aim_r, loc=<-9.3667 -0.0003 -0.0004>, rot=<0 0 0 1>}
 deform_pec_r {parent=deform_pec_base_r, loc=<0 0 0>, rot=<0 0 0 1>}
 aim_hood_rope_front_01_rt {parent=spine_04, loc=<1.1268 17.122 8.2914>, rot=<0.0377 -0.0635 0.9786 0.1923>}
 aim_hood_rope_front_01_lf {parent=spine_04, loc=<1.1263 17.122 -8.2914>, rot=<-0.0377 0.0635 0.9786 0.1923>}
 base_jacket_spine_cr {parent=spine_03, loc=<0 0 0>, rot=<0 0 0 1>}
 jacket_spine_cr {parent=base_jacket_spine_cr, loc=<0 0 0>, rot=<0 0 0 1>}
 base_coat_corr_front_lf {parent=jacket_spine_cr, loc=<0.3654 12.4701 -7.5828>, rot=<-0.0361 0.1437 0.9887 -0.0216>}
 aim_coat_corr_front_01_lf {parent=base_coat_corr_front_lf, loc=<0 0 0>, rot=<0 0 0 1>}
 coat_corr_front_01_lf {parent=aim_coat_corr_front_01_lf, loc=<0 0 0>, rot=<0 0 0 1>}
 base_coat_corr_front_rt {parent=jacket_spine_cr, loc=<0.3659 12.4701 7.5828>, rot=<0.0361 -0.1437 0.9887 -0.0216>}
 aim_coat_corr_front_01_rt {parent=base_coat_corr_front_rt, loc=<0 0 0>, rot=<0 0 0 1>}
 coat_corr_front_01_rt {parent=aim_coat_corr_front_01_rt, loc=<0 0 0>, rot=<0 0 0 1>}
 base_coat_cr {parent=spine_02, loc=<0 0 0>, rot=<0.0159 -0.9999 0 0>}
 coat_cr {parent=base_coat_cr, loc=<0 0 0>, rot=<0 0 0 1>}
 base_coat_back_lf {parent=coat_cr, loc=<1.3778 -7.4478 4.4925>, rot=<0.03 0.0096 0.0935 -0.9951>}
 offset_coat_back_lf {parent=base_coat_back_lf, loc=<0 0 0>, rot=<0 0 0 1>}
 dyn_coat_back_01_lf {parent=offset_coat_back_lf, loc=<0 0 0>, rot=<0 0 0 1>}
 dyn_coat_back_02_lf {parent=dyn_coat_back_01_lf, loc=<16.1681 0 0>, rot=<0 0 0 1>}
 dyn_coat_back_03_lf {parent=dyn_coat_back_02_lf, loc=<16.1681 0 0>, rot=<0 0 0 1>}
 dyn_coat_back_04_lf {parent=dyn_coat_back_03_lf, loc=<16.1681 0 0>, rot=<0 0 0 1>}
 base_skirt_rope_back_lf {parent=dyn_coat_back_04_lf, loc=<10.4061 -2.1868 -2.6724>, rot=<0.0827 0.0301 0.0844 -0.9925>}
 dyn_skirt_rope_back_01_lf {parent=base_skirt_rope_back_lf, loc=<0 0 0>, rot=<0 0 0 1>}
 dyn_skirt_rope_back_02_lf {parent=dyn_skirt_rope_back_01_lf, loc=<5.1172 -0.0628 0.0338>, rot=<0.0755 -0.0006 -0.0694 -0.9947>}
 dyn_skirt_rope_back_03_lf {parent=dyn_skirt_rope_back_02_lf, loc=<3.667 -0.0176 0.017>, rot=<-0.011 0.0088 -0.0116 -0.9998>}
 base_coat_back_side_lf {parent=coat_cr, loc=<1.3632 -4.9091 9.4161>, rot=<0.2329 0.021 0.0977 -0.9674>}
 offset_coat_back_side_lf {parent=base_coat_back_side_lf, loc=<0 0 0>, rot=<0 0 0 1>}
 dyn_coat_back_side_01_lf {parent=offset_coat_back_side_lf, loc=<0 0 0>, rot=<0 0 0 1>}
 dyn_coat_back_side_02_lf {parent=dyn_coat_back_side_01_lf, loc=<16.168 0 0>, rot=<0 0 0 1>}
 dyn_coat_back_side_03_lf {parent=dyn_coat_back_side_02_lf, loc=<16.168 0 0>, rot=<0 0 0 1>}
 dyn_coat_back_side_04_lf {parent=dyn_coat_back_side_03_lf, loc=<16.1681 0 0>, rot=<0 0 0 1>}
 base_coat_side_lf {parent=coat_cr, loc=<1.9953 -0.2004 12.9347>, rot=<0.5471 0.0418 0.115 -0.8281>}
 offset_coat_side_lf {parent=base_coat_side_lf, loc=<0 0 0>, rot=<0 0 0 1>}
 dyn_coat_side_01_lf {parent=offset_coat_side_lf, loc=<0 0 0>, rot=<0 0 0 1>}
 dyn_coat_side_02_lf {parent=dyn_coat_side_01_lf, loc=<16.1681 0 0>, rot=<0 0 0 1>}
 dyn_coat_side_03_lf {parent=dyn_coat_side_02_lf, loc=<16.1681 0 0>, rot=<0 0 0 1>}
 dyn_coat_side_04_lf {parent=dyn_coat_side_03_lf, loc=<16.1681 0 0>, rot=<0 0 0 1>}
 base_coat_front_side_lf {parent=coat_cr, loc=<2.3603 7.2851 12.9583>, rot=<0.8266 0.0376 0.0957 -0.5533>}
 offset_coat_front_side_lf {parent=base_coat_front_side_lf, loc=<0 0 0>, rot=<0 0 0 1>}
 dyn_coat_front_side_01_lf {parent=offset_coat_front_side_lf, loc=<0 0 0>, rot=<0 0 0 1>}
 dyn_coat_front_side_02_lf {parent=dyn_coat_front_side_01_lf, loc=<16.1681 0 0>, rot=<0 0 0 1>}
 dyn_coat_front_side_03_lf {parent=dyn_coat_front_side_02_lf, loc=<16.1681 0 0>, rot=<0 0 0 1>}
 dyn_coat_front_side_04_lf {parent=dyn_coat_front_side_03_lf, loc=<16.1681 0 0>, rot=<0 0 0 1>}
 base_coat_back_side_rt {parent=coat_cr, loc=<1.3629 -4.9091 -9.4161>, rot=<-0.2329 -0.021 0.0977 -0.9674>}
 offset_coat_back_side_rt {parent=base_coat_back_side_rt, loc=<0 0 0>, rot=<0 0 0 1>}
 dyn_coat_back_side_01_rt {parent=offset_coat_back_side_rt, loc=<0 0 0>, rot=<0 0 0 1>}
 dyn_coat_back_side_02_rt {parent=dyn_coat_back_side_01_rt, loc=<16.1684 0.0001 0>, rot=<0 0 0 1>}
 dyn_coat_back_side_03_rt {parent=dyn_coat_back_side_02_rt, loc=<16.168 0 0>, rot=<0 0 0 1>}
 dyn_coat_back_side_04_rt {parent=dyn_coat_back_side_03_rt, loc=<16.1682 0.0001 0>, rot=<0 0 0 1>}
 base_coat_back_rt {parent=coat_cr, loc=<1.378 -7.4478 -4.4925>, rot=<-0.03 -0.0096 0.0935 -0.9951>}
 offset_coat_back_rt {parent=base_coat_back_rt, loc=<0 0 0>, rot=<0 0 0 1>}
 dyn_coat_back_01_rt {parent=offset_coat_back_rt, loc=<0 0 0>, rot=<0 0 0 1>}
 dyn_coat_back_02_rt {parent=dyn_coat_back_01_rt, loc=<16.1679 -0.0001 0>, rot=<0 0 0 1>}
 dyn_coat_back_03_rt {parent=dyn_coat_back_02_rt, loc=<16.1681 0.0001 0>, rot=<0 0 0 1>}
 dyn_coat_back_04_rt {parent=dyn_coat_back_03_rt, loc=<16.1681 0 0>, rot=<0 0 0 1>}
 base_skirt_rope_back_rt {parent=dyn_coat_back_04_rt, loc=<10.406 -2.1868 2.6724>, rot=<-0.1072 -0.0395 0.0873 -0.9896>}
 dyn_skirt_rope_back_01_rt {parent=base_skirt_rope_back_rt, loc=<0 0 0>, rot=<0 0 0 1>}
 dyn_skirt_rope_back_02_rt {parent=dyn_skirt_rope_back_01_rt, loc=<5.1396 0.0073 0.0322>, rot=<-0.028 0.0053 -0.0657 -0.9974>}
 dyn_skirt_rope_back_03_rt {parent=dyn_skirt_rope_back_02_rt, loc=<3.6156 0.0231 -0.0017>, rot=<0.007 -0.0019 -0.0233 -0.9997>}
 base_coat_side_rt {parent=coat_cr, loc=<1.9956 -0.2004 -12.9347>, rot=<0.0418 -0.5471 0.8281 0.115>}
 offset_coat_side_rt {parent=base_coat_side_rt, loc=<0 0 0>, rot=<0 0 1 0>}
 dyn_coat_side_01_rt {parent=offset_coat_side_rt, loc=<0 0 0>, rot=<0 0 0 1>}
 dyn_coat_side_02_rt {parent=dyn_coat_side_01_rt, loc=<16.1677 -0.0001 0>, rot=<0 0 0 1>}
 dyn_coat_side_03_rt {parent=dyn_coat_side_02_rt, loc=<16.1681 0 0>, rot=<0 0 0 1>}
 dyn_coat_side_04_rt {parent=dyn_coat_side_03_rt, loc=<16.1681 0 0>, rot=<0 0 0 1>}
 base_coat_front_side_rt {parent=coat_cr, loc=<2.3603 7.2851 -12.9583>, rot=<-0.8266 -0.0376 0.0957 -0.5533>}
 offset_coat_front_side_rt {parent=base_coat_front_side_rt, loc=<0 0 0>, rot=<0 0 0 1>}
 dyn_coat_front_side_01_rt {parent=offset_coat_front_side_rt, loc=<0 0 0>, rot=<0 0 0 1>}
 dyn_coat_front_side_02_rt {parent=dyn_coat_front_side_01_rt, loc=<16.1681 0 0>, rot=<0 0 0 1>}
 dyn_coat_front_side_03_rt {parent=dyn_coat_front_side_02_rt, loc=<16.1681 0 0>, rot=<0 0 0 1>}
 dyn_coat_front_side_04_rt {parent=dyn_coat_front_side_03_rt, loc=<16.1681 0 0>, rot=<0 0 0 1>}
 orient_coat_front_lf {parent=coat_cr, loc=<2.4 12.3123 8.2061>, rot=<0.7361 0.6769 0 0>}
 base_coat_front_lf {parent=orient_coat_front_lf, loc=<0 0 0>, rot=<-0.0867 -0.1596 -0.6609 -0.7282>}
 offset_coat_front_lf {parent=base_coat_front_lf, loc=<0 0 0>, rot=<0 0 0 1>}
 dyn_coat_front_01_lf {parent=offset_coat_front_lf, loc=<0 0 0>, rot=<0 0 0 1>}
 dyn_coat_front_02_lf {parent=dyn_coat_front_01_lf, loc=<8.0861 0 0>, rot=<0 0 0.0253 -0.9997>}
 dyn_coat_front_03_lf {parent=dyn_coat_front_02_lf, loc=<8.0576 -0.4022 0>, rot=<0 0 0 1>}
 dyn_coat_front_04_lf {parent=dyn_coat_front_03_lf, loc=<16.1517 0 0>, rot=<0 0 0 1>}
 base_zipper_front_lf {parent=dyn_coat_front_04_lf, loc=<13.0607 -0.8342 2.2658>, rot=<-0.0381 -0.009 0.0031 -0.9992>}
 dyn_zipper_front_lf {parent=base_zipper_front_lf, loc=<0 0 0>, rot=<0 0 0 1>}
 dyn_coat_front_05_lf {parent=dyn_coat_front_04_lf, loc=<16.1517 0 0>, rot=<0 0 0 1>}
 orient_coat_front_rt {parent=coat_cr, loc=<2.3996 12.3123 -8.2061>, rot=<0.7361 0.6769 0 0>}
 base_coat_front_rt {parent=orient_coat_front_rt, loc=<0 0 0>, rot=<0.0867 0.1596 -0.6609 -0.7282>}
 offset_coat_front_rt {parent=base_coat_front_rt, loc=<0 0 0>, rot=<0 0 0 1>}
 dyn_coat_front_01_rt {parent=offset_coat_front_rt, loc=<0 0 0>, rot=<0 0 0 1>}
 dyn_coat_front_02_rt {parent=dyn_coat_front_01_rt, loc=<8.086 0 0>, rot=<0 0 0.0021 -1>}
 dyn_coat_front_03_rt {parent=dyn_coat_front_02_rt, loc=<8.0581 -0.4023 0>, rot=<0 0 0.0346 -0.9994>}
 dyn_coat_front_04_rt {parent=dyn_coat_front_03_rt, loc=<16.1517 0 0>, rot=<0 0 0 1>}
 dyn_coat_front_05_rt {parent=dyn_coat_front_04_rt, loc=<16.1517 0 0>, rot=<0 0 0 1>}
 tgt_coat_front_side_rt {parent=spine_02, loc=<-0.2632 13.0327 26.5188>, rot=<0.1044 -0.5518 0.8271 0.0245>}
 tgt_coat_back_rt {parent=spine_02, loc=<2.1113 -22.0454 5.3578>, rot=<0.1093 -0.9935 0.0301 0.0091>}
 tgt_coat_back_side_rt {parent=spine_02, loc=<2.1756 -17.885 16.0868>, rot=<0.113 -0.9657 0.2332 0.0173>}
 tgt_coat_front_side_lf {parent=spine_02, loc=<-0.2632 13.0327 -26.5188>, rot=<-0.1044 0.5518 0.8271 0.0245>}
 tgt_coat_back_side_lf {parent=spine_02, loc=<2.1752 -17.885 -16.0868>, rot=<0.113 -0.9657 -0.2332 -0.0173>}
 tgt_coat_back_lf {parent=spine_02, loc=<2.1115 -22.0454 -5.3578>, rot=<0.1093 -0.9935 -0.0301 -0.0091>}
 tgt_coat_front_lf {parent=spine_02, loc=<-2.7457 26.2966 -13.2434>, rot=<-0.0615 0.1708 0.9833 -0.0092>}
 tgt_coat_front_rt {parent=spine_02, loc=<-2.7453 26.2966 13.2434>, rot=<0.0615 -0.1708 0.9833 -0.0092>}
 tgt_coat_side_rt {parent=spine_02, loc=<1.7169 -5.7494 26.3298>, rot=<0.1281 -0.8261 0.5477 0.0331>}
 tgt_coat_side_lf {parent=spine_02, loc=<1.7168 -5.7495 -26.3299>, rot=<0.1281 -0.8261 -0.5477 -0.0331>}
 thigh_l {parent=pelvis, loc=<-2.5693 -0.0109 -7.7501>, rot=<0.0012 0.0572 0.0155 -0.9982>}
 calf_l {parent=thigh_l, loc=<-39.0484 0 0>, rot=<0.0006 -0.0161 -0.0391 -0.9991>}
 foot_l {parent=calf_l, loc=<-42.561 0.1077 0.4292>, rot=<0.0418 -0.0455 0.0343 -0.9975>}
 ball_l {parent=foot_l, loc=<-6.1599 13.8264 0.6806>, rot=<0.0001 -0.0001 -0.7063 -0.7079>}
 calf_twist_02_l {parent=calf_l, loc=<-18.8064 -0.0967 0.0397>, rot=<0.0237 0.0095 -0.0035 -0.9997>}
 calf_twist_01_l {parent=calf_l, loc=<-35.813 -0.0967 0.0397>, rot=<0.0237 0.0095 -0.0035 -0.9997>}
 deform_knee_in_base_l {parent=calf_l, loc=<3.6268 -5.2446 -0.2814>, rot=<-0.0017 -0.0216 -0.0001 -0.9998>}
 deform_knee_in_l {parent=deform_knee_in_base_l, loc=<0 0 0>, rot=<0 0 0 1>}
 deform_knee_out_base_l {parent=calf_l, loc=<2.8134 3.7791 -0.4614>, rot=<-0.0017 -0.0216 -0.0001 -0.9998>}
 deform_knee_out_l {parent=deform_knee_out_base_l, loc=<0 0 0>, rot=<0 0 0 1>}
 thigh_twist_01_l {parent=thigh_l, loc=<-21.6303 0 0>, rot=<0.0013 0 0 -1>}
 deform_groin_base_l {parent=thigh_l, loc=<6.7459 5.6776 -1.952>, rot=<0 -0.0217 0 -0.9998>}
 deform_groin_l {parent=deform_groin_base_l, loc=<0 0 0>, rot=<0 0 0 1>}
 thigh_r {parent=pelvis, loc=<-2.5693 -0.0109 7.7501>, rot=<0.0572 -0.0012 0.9982 0.0155>}
 calf_r {parent=thigh_r, loc=<39.0484 0 0>, rot=<0.0006 -0.0161 -0.0391 -0.9991>}
 foot_r {parent=calf_r, loc=<42.561 -0.1077 -0.4292>, rot=<0.0418 -0.0455 0.0343 -0.9975>}
 ball_r {parent=foot_r, loc=<6.1599 -13.8264 -0.6806>, rot=<0.0001 -0.0001 -0.7063 -0.7079>}
 calf_twist_02_r {parent=calf_r, loc=<18.8064 0.0967 -0.0397>, rot=<0.0237 0.0095 -0.0035 -0.9997>}
 calf_twist_01_r {parent=calf_r, loc=<35.813 0.0967 -0.0397>, rot=<0.0237 0.0095 -0.0035 -0.9997>}
 deform_knee_in_base_r {parent=calf_r, loc=<-3.6267 5.2446 0.2814>, rot=<-0.0017 -0.0216 -0.0001 -0.9998>}
 deform_knee_in_r {parent=deform_knee_in_base_r, loc=<0 0 0>, rot=<0 0 0 1>}
 deform_knee_out_base_r {parent=calf_r, loc=<-2.8134 -3.7791 0.4615>, rot=<-0.0017 -0.0216 -0.0001 -0.9998>}
 deform_knee_out_r {parent=deform_knee_out_base_r, loc=<0 0 0>, rot=<0 0 0 1>}
 thigh_twist_01_r {parent=thigh_r, loc=<21.6303 0 0>, rot=<0.0013 0 0 -1>}
 deform_groin_base_r {parent=thigh_r, loc=<-6.7459 -5.6776 1.952>, rot=<0 -0.0217 0 -0.9998>}
 deform_groin_r {parent=deform_groin_base_r, loc=<0 0 0>, rot=<0 0 0 1>}
 deform_glute_base_l {parent=pelvis, loc=<-2.4054 -9.0213 -9.1245>, rot=<0.0172 0.031 -0.4477 -0.8935>}
 deform_glute_l {parent=deform_glute_base_l, loc=<-0.0161 -0.0403 0>, rot=<0 0 0 1>}
 deform_glute_base_r {parent=pelvis, loc=<-2.4052 -9.0213 9.1245>, rot=<0.031 -0.0172 0.8935 -0.4477>}
 deform_glute_r {parent=deform_glute_base_r, loc=<0.0161 0.0403 0>, rot=<0 0 0 1>}
 pelvis_cloth_helper {parent=pelvis, loc=<0 0 0>, rot=<0 0 0 1>}
 floor {parent=root, loc=<0 0 0>, rot=<0 0 0 1>}
 cloth_groung_collision {parent=root, loc=<0 0 0>, rot=<0 0 0 1>}
 cloth_collision_helper {parent=root, loc=<0 0 0>, rot=<0 0 0 1>}
]
variables [
 CRisCrouching {type=bool, default=False}
 IsClothOn {type=bool, default=False}
 IsCrouch {type=bool, default=False}
 IsCrouchWalking {type=bool, default=False}
 IsMelee {type=bool, default=False}
 IsMeleeNotCrouch {type=bool, default=False}
 IsOnAir {type=bool, default=False}
 IsRail {type=bool, default=False}
]
