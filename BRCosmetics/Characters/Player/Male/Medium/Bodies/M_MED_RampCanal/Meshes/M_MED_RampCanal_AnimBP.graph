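# Post-process-style AnimBlueprint for the M_MED_RampCanal player cosmetic body, adding secondary motion to its coat, hood, zipper and ropes.
# It copies the parent mesh pose into a Control Rig fed crouch, melee, rail, skydive and cloth flags.
# Constraints and bool-blended bone offsets then reshape the coat before physics.
# RigidBody simulates the coat in pelvis space, switched by bClothEnabled, feeding chained AnimDynamics for smaller accessories.
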AnimBlueprint /BRCosmetics/Characters/Player/Male/Medium/Bodies/M_MED_RampCanal/Meshes/M_MED_RampCanal_AnimBP
target_skeleton=&/M_MED_RampCanal_Skeleton
graphs [
 AnimGraph {nodes [
   Output Pose(Result=Use cached pose 'Dynamics_CACHE'()) {layerGroup=None}
   n1 = Control Rig!(
    Source=Copy Pose From Mesh() {bUseAttachedParent=true}
    Alpha
    CRisCrouching:bool=Get bIsCrouching()
    IsCrouch:bool=Get CrouchIDLE() -> (CrouchIDLE:bool)
    IsCrouchWalking:bool=Get IsCrouchWalking() -> (IsCrouchWalking:bool)
    IsMelee:bool=Get bIsPlayingMeleeAnim()
    IsMeleeNotCrouch:bool=Get IsMeleeNotCrouching() -> (IsMeleeNotCrouching:bool)
    IsRail:bool=Get bIsOnGrindRail()
    IsClothOn:bool=Get bClothEnabled()
    IsOnAir:bool=Get bIsSkyDiving()
   ) -> (Pose) {controlRigAssetReference {blueprintRigClass=&/M_MED_RampCanal_CtrlRig.M_MED_RampCanal_CtrlRig_C}}
   Save cached pose 'ControlRig_CACHE'(Pose=@n1) {cacheName=ControlRig_CACHE}
   n2 = "Constraint - base_skirt_rope_back_lf (PPPP)"!(
    ComponentPose=Local To Component(LocalPose=Use cached pose 'ControlRig_CACHE'())
    Alpha=1
    ConstraintWeights_0=0.8027
    ConstraintWeights_1:float=0.188
    ConstraintWeights_2:float=0.0047
    ConstraintWeights_3:float=0.0046
   ) -> (Pose) {
    boneToModify {boneName=base_skirt_rope_back_lf}
    constraintSetup [
     {targetBone {boneName=dyn_coat_back_04_lf}}
     {targetBone {boneName=dyn_coat_back_03_lf}}
     {targetBone {boneName=dyn_coat_back_side_04_lf}}
     {targetBone {boneName=dyn_coat_back_02_lf}}
    ]
    constraintWeights [1,1,1,1]
   }
   n3 = "Constraint - base_skirt_rope_back_rt (PPPP)"!(
    ComponentPose=@n2
    Alpha=1
    ConstraintWeights_0=0.8123
    ConstraintWeights_1:float=0.1822
    ConstraintWeights_2:float=0.0034
    ConstraintWeights_3:float=0.002
   ) -> (Pose) {
    boneToModify {boneName=base_skirt_rope_back_rt}
    constraintSetup [
     {targetBone {boneName=dyn_coat_back_04_rt}}
     {targetBone {boneName=dyn_coat_back_03_rt}}
     {targetBone {boneName=dyn_coat_back_02_rt}}
     {targetBone {boneName=dyn_coat_back_side_04_rt}}
    ]
    constraintWeights [1,1,1,1]
   }
   n4 = "Transform (Modify) Bone - Bone: offset_coat_side_rt"!(ComponentPose=@n3, bAlphaBoolEnabled:bool=True, Translation="0,0,0", Scale="1,1,1") -> (Pose) {
    alphaInputType=Bool
    rotationMode=BMM_Additive
    rotationSpace=BCS_BoneSpace
    alphaBoolBlend {blendInTime=0.4, blendOutTime=0.4}
    boneToModify {boneName=offset_coat_side_rt}
    rotation {yaw=-20}
   }
   n5 = "Transform (Modify) Bone - Bone: offset_coat_back_rt"!(ComponentPose=@n4, bAlphaBoolEnabled:bool=True, Translation="0,0,0", Scale="1,1,1") -> (Pose) {
    alphaInputType=Bool
    rotationMode=BMM_Additive
    rotationSpace=BCS_BoneSpace
    alphaBoolBlend {blendInTime=0.4, blendOutTime=0.4}
    boneToModify {boneName=offset_coat_back_rt}
    rotation {yaw=-20}
   }
   n6 = "Transform (Modify) Bone - Bone: offset_coat_back_lf"!(ComponentPose=@n5, bAlphaBoolEnabled:bool=True, Translation="0,0,0", Scale="1,1,1") -> (Pose) {
    alphaInputType=Bool
    rotationMode=BMM_Additive
    rotationSpace=BCS_BoneSpace
    alphaBoolBlend {blendInTime=0.4, blendOutTime=0.4}
    boneToModify {boneName=offset_coat_back_lf}
    rotation {yaw=-20}
   }
   n7 = "Transform (Modify) Bone - Bone: offset_coat_front_side_rt"!(ComponentPose=@n6, bAlphaBoolEnabled:bool=True, Translation="0,0,0", Scale="1,1,1") -> (Pose) {
    alphaInputType=Bool
    rotationMode=BMM_Additive
    rotationSpace=BCS_BoneSpace
    alphaBoolBlend {blendInTime=0.4, blendOutTime=0.4}
    boneToModify {boneName=offset_coat_front_side_rt}
    rotation {yaw=-20}
   }
   n8 = "Transform (Modify) Bone - Bone: offset_coat_back_side_lf"!(ComponentPose=@n7, bAlphaBoolEnabled:bool=True, Translation="0,0,0", Scale="1,1,1") -> (Pose) {
    alphaInputType=Bool
    rotationMode=BMM_Additive
    rotationSpace=BCS_BoneSpace
    alphaBoolBlend {blendInTime=0.4, blendOutTime=0.4}
    boneToModify {boneName=offset_coat_back_side_lf}
    rotation {yaw=-20}
   }
   n9 = "Transform (Modify) Bone - Bone: offset_coat_side_lf"!(ComponentPose=@n8, bAlphaBoolEnabled:bool=True, Translation="0,0,0", Scale="1,1,1") -> (Pose) {
    alphaInputType=Bool
    rotationMode=BMM_Additive
    rotationSpace=BCS_BoneSpace
    alphaBoolBlend {blendInTime=0.4, blendOutTime=0.4}
    boneToModify {boneName=offset_coat_side_lf}
    rotation {yaw=-20}
   }
   n10 = "Transform (Modify) Bone - Bone: offset_coat_front_side_rt"(ComponentPose=@n9, bAlphaBoolEnabled=True, Translation="0,0,0", Scale="1,1,1") {
    alphaInputType=Bool
    rotationMode=BMM_Additive
    rotationSpace=BCS_BoneSpace
    alphaBoolBlend {blendInTime=0.4, blendOutTime=0.4}
    boneToModify {boneName=offset_coat_front_side_rt}
    rotation {yaw=-20}
   }
   n11 = "Transform (Modify) Bone - Bone: offset_coat_front_rt"!(ComponentPose=@n10, bAlphaBoolEnabled:bool=True, Translation="0,-7,0", Scale="1,1,1") -> (Pose) {
    alphaInputType=Bool
    rotationMode=BMM_Additive
    rotationSpace=BCS_BoneSpace
    translationMode=BMM_Additive
    translationSpace=BCS_BoneSpace
    alphaBoolBlend {blendInTime=0.4, blendOutTime=0.4}
    boneToModify {boneName=offset_coat_front_rt}
    rotation {yaw=-19.8806}
    translation {y=-7}
   }
   n12 = "Transform (Modify) Bone - Bone: offset_coat_front_lf"!(ComponentPose=@n11, bAlphaBoolEnabled:bool=True, Translation="0,-5,0", Scale="1,1,1") -> (Pose) {
    alphaInputType=Bool
    rotationMode=BMM_Additive
    rotationSpace=BCS_BoneSpace
    translationMode=BMM_Additive
    translationSpace=BCS_BoneSpace
    alphaBoolBlend {blendInTime=0.4, blendOutTime=0.4}
    boneToModify {boneName=offset_coat_front_lf}
    rotation=<rot -9.8729 -10.1526 1.7073>
    translation {y=-5}
   }
   n13 = "Transform (Modify) Bone - Bone: coat_cr"!(ComponentPose=@n12, bAlphaBoolEnabled:bool=True, Scale="1,1,1") -> (Pose) {
    alphaInputType=Bool
    rotationMode=BMM_Additive
    rotationSpace=BCS_BoneSpace
    translationMode=BMM_Additive
    translationSpace=BCS_BoneSpace
    alphaBoolBlend {blendInTime=0.4, blendOutTime=0.4}
    boneToModify {boneName=coat_cr}
    rotation=<rot 5.8497 19.8269 -14.6016>
   }
   n14 = "Transform (Modify) Bone - Bone: offset_coat_front_lf"(ComponentPose=@n13, bAlphaBoolEnabled=True, Translation="7.047,-1.522,-0.318", Scale="1,1,1") {
    alphaInputType=Bool
    rotationMode=BMM_Replace
    rotationSpace=BCS_BoneSpace
    translationMode=BMM_Replace
    translationSpace=BCS_BoneSpace
    alphaBoolBlend {blendInTime=0.4, blendOutTime=0.4}
    boneToModify {boneName=offset_coat_front_lf}
    rotation=<rot -30.5512 9.6795 -8.0345>
    translation=<7.047 -1.522 -0.318>
   }
   n15 = "Transform (Modify) Bone - Bone: offset_coat_front_side_lf"!(ComponentPose=@n14, bAlphaBoolEnabled:bool=True, Translation="6.806,8.770,1.101", Scale="1,1,1") -> (Pose) {
    alphaInputType=Bool
    rotationMode=BMM_Additive
    rotationSpace=BCS_BoneSpace
    translationMode=BMM_Additive
    alphaBoolBlend {blendInTime=0.4, blendOutTime=0.4}
    boneToModify {boneName=offset_coat_front_side_lf}
    rotation=<rot -13.7354 20.8284 6.0542>
    translation=<6.806 8.77 1.101>
   }
   n16 = "Transform (Modify) Bone - Bone: offset_coat_side_lf"(ComponentPose=@n15, bAlphaBoolEnabled=True, Translation="12.689,1.380,0.466", Scale="1,1,1") {
    alphaInputType=Bool
    rotationMode=BMM_Additive
    rotationSpace=BCS_BoneSpace
    translationMode=BMM_Additive
    alphaBoolBlend {blendInTime=0.4, blendOutTime=0.4}
    boneToModify {boneName=offset_coat_side_lf}
    rotation=<rot -18.0595 -10.0656 26.4218>
    translation=<12.689 1.38 0.466>
   }
   n17 = "Transform (Modify) Bone - Bone: base_coat_back_lf"!(ComponentPose=@n16, bAlphaBoolEnabled:bool=True, Scale="1,1,1") -> (Pose) {
    alphaInputType=Bool
    rotationMode=BMM_Additive
    rotationSpace=BCS_BoneSpace
    translationMode=BMM_Additive
    translationSpace=BCS_BoneSpace
    alphaBoolBlend {blendInTime=0.4, blendOutTime=0.4}
    boneToModify {boneName=base_coat_back_lf}
    rotation=<rot 0.0567 -24.9958 0.0384>
    translation {x=10}
   }
   n18 = "Transform (Modify) Bone - Bone: offset_coat_back_side_lf"!(ComponentPose=@n17, bAlphaBoolEnabled=True, Scale="1,1,1") -> (Pose) {
    alphaInputType=Bool
    rotationMode=BMM_Additive
    rotationSpace=BCS_BoneSpace
    translationMode=BMM_Additive
    translationSpace=BCS_BoneSpace
    alphaBoolBlend {blendInTime=0.4, blendOutTime=0.4}
    boneToModify {boneName=offset_coat_back_side_lf}
    rotation=<rot -0.6738 -30.004 0.5046>
    translation=<0.3237 -6.4744 0.0778>
   }
   n19 = "Transform (Modify) Bone - Bone: offset_coat_back_rt"!(ComponentPose=@n18, bAlphaBoolEnabled=True, Scale="1,1,1") -> (Pose) {
    alphaInputType=Bool
    rotationMode=BMM_Additive
    rotationSpace=BCS_BoneSpace
    translationMode=BMM_Additive
    translationSpace=BCS_BoneSpace
    alphaBoolBlend {blendInTime=0.4, blendOutTime=0.4}
    boneToModify {boneName=offset_coat_back_rt}
    rotation=<rot 0 -19.9976 -0.0125>
    translation=<8.3402 -4.5626 0.0096>
   }
   n20 = "Transform (Modify) Bone - Bone: offset_coat_back_side_rt"!(ComponentPose=@n19, bAlphaBoolEnabled:bool=True, Scale="1,1,1") -> (Pose) {
    alphaInputType=Bool
    rotationMode=BMM_Additive
    rotationSpace=BCS_BoneSpace
    translationMode=BMM_Additive
    translationSpace=BCS_BoneSpace
    alphaBoolBlend {blendInTime=0.4, blendOutTime=0.4}
    boneToModify {boneName=offset_coat_back_side_rt}
    rotation=<rot -0.3144 -49.9714 -0.4762>
   }
   n21 = "Transform (Modify) Bone - Bone: floor"!(ComponentPose=@n20, bAlphaBoolEnabled:bool=True, Translation="0,0,-30", Scale="1,1,1") -> (Pose) {
    alphaInputType=Bool
    rotationMode=BMM_Additive
    rotationSpace=BCS_BoneSpace
    translationMode=BMM_Additive
    translationSpace=BCS_BoneSpace
    alphaBoolBlend {blendInTime=0.4, blendOutTime=0.4}
    boneToModify {boneName=floor}
    translation {z=-30}
   }
   n22 = "Transform (Modify) Bone - Bone: offset_coat_front_rt"(ComponentPose=@n21, bAlphaBoolEnabled=True, Translation="0,0,-30", Scale="1,1,1") {
    alphaInputType=Bool
    rotationMode=BMM_Additive
    rotationSpace=BCS_BoneSpace
    translationMode=BMM_Additive
    translationSpace=BCS_BoneSpace
    alphaBoolBlend {blendInTime=0.4, blendOutTime=0.4}
    boneToModify {boneName=offset_coat_front_rt}
    rotation=<rot 49.96 -1.3987 -1.6969>
   }
   n23 = "Transform (Modify) Bone - Bone: offset_coat_front_side_rt"(ComponentPose=@n22, bAlphaBoolEnabled=True, Translation="0,0,-30", Scale="1,1,1") {
    alphaInputType=Bool
    rotationMode=BMM_Additive
    rotationSpace=BCS_BoneSpace
    translationMode=BMM_Additive
    translationSpace=BCS_BoneSpace
    alphaBoolBlend {blendInTime=0.4, blendOutTime=0.4}
    boneToModify {boneName=offset_coat_front_side_rt}
    rotation=<rot 33.6697 -35.1951 -0.4187>
   }
   n24 = "Transform (Modify) Bone - Bone: offset_coat_side_rt"(ComponentPose=@n23, bAlphaBoolEnabled=True, Translation="0,0,0", Scale="1,1,1") {
    alphaInputType=Bool
    rotationMode=BMM_Additive
    rotationSpace=BCS_BoneSpace
    translationMode=BMM_Additive
    translationSpace=BCS_BoneSpace
    alphaBoolBlend {blendInTime=0.4, blendOutTime=0.4}
    boneToModify {boneName=offset_coat_side_rt}
    rotation=<rot -0.0296 -24.9991 0.1804>
   }
   n25 = "Transform (Modify) Bone - Bone: base_coat_back_side_lf"!(ComponentPose=@n24, bAlphaBoolEnabled:bool=True, Translation="0,0,0", Scale="1,1,1") -> (Pose) {
    alphaInputType=Bool
    rotationMode=BMM_Additive
    rotationSpace=BCS_BoneSpace
    alphaBoolBlend {blendInTime=0.4, blendOutTime=0.2}
    boneToModify {boneName=base_coat_back_side_lf}
    rotation=<rot -10 -69.9995 -0.0003>
   }
   n26 = "Transform (Modify) Bone - Bone: base_coat_back_lf"!(ComponentPose=@n25, bAlphaBoolEnabled=True, Translation:Vector="0,0,0", Scale="1,1,1") -> (Pose) {
    alphaInputType=Bool
    rotationMode=BMM_Additive
    rotationSpace=BCS_BoneSpace
    alphaBoolBlend {blendInTime=0.4, blendOutTime=0.2}
    boneToModify {boneName=base_coat_back_lf}
    rotation=<rot -5 -74.9998 0>
   }
   n27 = "Transform (Modify) Bone - Bone: base_coat_back_side_rt"!(ComponentPose=@n26, bAlphaBoolEnabled:bool=True, Translation="0,0,0", Scale="1,1,1") -> (Pose) {
    alphaInputType=Bool
    rotationMode=BMM_Additive
    rotationSpace=BCS_BoneSpace
    alphaBoolBlend {blendInTime=0.4, blendOutTime=0.2}
    boneToModify {boneName=base_coat_back_side_rt}
    rotation=<rot 6.6947 -48.6308 6.8122>
   }
   n28 = "Transform (Modify) Bone - Bone: base_coat_back_rt"!(ComponentPose=@n27, bAlphaBoolEnabled:bool=True, Translation="0,0,0", Scale="1,1,1") -> (Pose) {
    alphaInputType=Bool
    rotationMode=BMM_Additive
    rotationSpace=BCS_BoneSpace
    alphaBoolBlend {blendInTime=0.4, blendOutTime=0.2}
    boneToModify {boneName=base_coat_back_rt}
    rotation=<rot -2.3973 -65.3438 4.3817>
   }
   n29 = "Transform (Modify) Bone - Bone: base_coat_side_rt"!(ComponentPose=@n28, bAlphaBoolEnabled:bool=True, Translation="0,0,0", Scale="1,1,1") -> (Pose) {
    alphaInputType=Bool
    rotationMode=BMM_Additive
    rotationSpace=BCS_BoneSpace
    alphaBoolBlend {blendInTime=0.4, blendOutTime=0.2}
    boneToModify {boneName=base_coat_side_rt}
    rotation=<rot -30 -19.9999 -0.0001>
   }
   n30 = "Transform (Modify) Bone - Bone: base_coat_side_lf"!(ComponentPose=@n29, bAlphaBoolEnabled:bool=True, Translation="0,0,0", Scale="1,1,1") -> (Pose) {
    alphaInputType=Bool
    rotationMode=BMM_Additive
    rotationSpace=BCS_BoneSpace
    alphaBoolBlend {blendInTime=0.4, blendOutTime=0.2}
    boneToModify {boneName=base_coat_side_lf}
    rotation=<rot -39.8182 -36.0231 1.6827>
   }
   n31 = "Transform (Modify) Bone - Bone: offset_coat_front_rt"(ComponentPose=@n30, bAlphaBoolEnabled=True, Translation="3.556,5.800,-0.061", Scale="1,1,1") {
    alphaInputType=Bool
    rotationMode=BMM_Additive
    rotationSpace=BCS_BoneSpace
    translationMode=BMM_Additive
    translationSpace=BCS_BoneSpace
    alphaBoolBlend {blendInTime=0.4, blendOutTime=0.2}
    boneToModify {boneName=offset_coat_front_rt}
    rotation=<rot 15.4107 26.9158 -29.4025>
    translation=<3.556 5.8 -0.061>
   }
   n32 = "Transform (Modify) Bone - Bone: offset_coat_front_lf"(ComponentPose=@n31, bAlphaBoolEnabled=True, Translation="5.881,4.380,-2.287", Scale="1,1,1") {
    alphaInputType=Bool
    rotationMode=BMM_Additive
    rotationSpace=BCS_BoneSpace
    translationMode=BMM_Additive
    translationSpace=BCS_BoneSpace
    alphaBoolBlend {blendInTime=0.4, blendOutTime=0.2}
    boneToModify {boneName=offset_coat_front_lf}
    rotation=<rot -42.3277 16.0866 -45.5213>
    translation=<5.881 4.3801 -2.287>
   }
   n33 = "Transform (Modify) Bone - Bone: offset_coat_front_side_lf"(ComponentPose=@n32, bAlphaBoolEnabled=True, Translation="5.653,-2.598,-1.305", Scale="1,1,1") {
    alphaInputType=Bool
    rotationMode=BMM_Additive
    rotationSpace=BCS_BoneSpace
    translationMode=BMM_Additive
    translationSpace=BCS_BoneSpace
    alphaBoolBlend {blendInTime=0.4, blendOutTime=0.2}
    boneToModify {boneName=offset_coat_front_side_lf}
    rotation=<rot -59.9998 0 -40.0001>
    translation=<5.653 -2.598 -1.305>
   }
   n34 = "Transform (Modify) Bone - Bone: offset_coat_front_side_rt"(ComponentPose=@n33, bAlphaBoolEnabled=True, Translation="5.266,-0.985,4.321", Scale="1,1,1") {
    alphaInputType=Bool
    rotationMode=BMM_Additive
    rotationSpace=BCS_BoneSpace
    translationMode=BMM_Additive
    translationSpace=BCS_BoneSpace
    alphaBoolBlend {blendInTime=0.4, blendOutTime=0.2}
    boneToModify {boneName=offset_coat_front_side_rt}
    rotation=<rot 35.9038 -5.2279 21.5568>
    translation=<5.266 -0.985 4.321>
   }
   n35 = "Transform (Modify) Bone - Bone: coat_corr_front_01_rt"!(ComponentPose=@n34, bAlphaBoolEnabled:bool=True, Translation="0,0,0", Scale="1,1,1") -> (Pose) {
    alphaInputType=Bool
    rotationMode=BMM_Additive
    rotationSpace=BCS_BoneSpace
    translationMode=BMM_Additive
    translationSpace=BCS_BoneSpace
    alphaBoolBlend {blendInTime=0.4, blendOutTime=0.2}
    boneToModify {boneName=coat_corr_front_01_rt}
    rotation=<rot 33.5111 -8.0762 -36.6573>
   }
   n36 = "Transform (Modify) Bone - Bone: coat_corr_front_01_lf"!(ComponentPose=@n35, bAlphaBoolEnabled:bool=True, Translation="0.553,-1.691,1.667", Scale="1,1,1") -> (Pose) {
    alphaInputType=Bool
    rotationMode=BMM_Additive
    rotationSpace=BCS_BoneSpace
    translationMode=BMM_Additive
    translationSpace=BCS_BoneSpace
    alphaBoolBlend {blendInTime=0.4, blendOutTime=0.2}
    boneToModify {boneName=coat_corr_front_01_lf}
    rotation=<rot 16.0322 58.571 8.8069>
    translation=<0.553 -1.691 1.667>
   }
   n37 = "Transform (Modify) Bone - Bone: base_coat_back_side_lf"(ComponentPose=@n36, bAlphaBoolEnabled=True, Translation="0,0,0", Scale="1,1,1") {
    alphaInputType=Bool
    rotationMode=BMM_Additive
    rotationSpace=BCS_BoneSpace
    alphaBoolBlend {blendInTime=0.5, blendOutTime=0.2}
    boneToModify {boneName=base_coat_back_side_lf}
    rotation=<rot -9.1528 -45.8302 22.2163>
   }
   n38 = "Transform (Modify) Bone - Bone: base_coat_back_lf"(ComponentPose=@n37, bAlphaBoolEnabled=True, Translation="0,0,0", Scale="1,1,1") {
    alphaInputType=Bool
    rotationMode=BMM_Additive
    rotationSpace=BCS_BoneSpace
    alphaBoolBlend {blendInTime=0.5, blendOutTime=0.2}
    boneToModify {boneName=base_coat_back_lf}
    rotation=<rot -14.4578 -43.7765 23.372>
   }
   n39 = "Transform (Modify) Bone - Bone: base_coat_back_side_rt"(ComponentPose=@n38, bAlphaBoolEnabled=True, Translation="-1.159,-1.657,0.000", Scale="1,1,1") {
    alphaInputType=Bool
    rotationMode=BMM_Additive
    rotationSpace=BCS_BoneSpace
    translationMode=BMM_Additive
    translationSpace=BCS_BoneSpace
    alphaBoolBlend {blendInTime=0.5, blendOutTime=0.2}
    boneToModify {boneName=base_coat_back_side_rt}
    rotation=<rot 0.1474 -80.2729 30.0892>
    translation=<-1.159 -1.657>
   }
   n40 = "Transform (Modify) Bone - Bone: base_coat_back_rt"(ComponentPose=@n39, bAlphaBoolEnabled=True, Translation="0,0,0", Scale="1,1,1") {
    alphaInputType=Bool
    rotationMode=BMM_Additive
    rotationSpace=BCS_BoneSpace
    alphaBoolBlend {blendInTime=0.5, blendOutTime=0.2}
    boneToModify {boneName=base_coat_back_rt}
    rotation=<rot -8.6745 -67.0595 35.8803>
   }
   n41 = "Transform (Modify) Bone - Bone: base_coat_side_rt"(ComponentPose=@n40, bAlphaBoolEnabled=True, Translation="-0.046,0.236,-0.311", Scale="1,1,1") {
    alphaInputType=Bool
    rotationMode=BMM_Additive
    rotationSpace=BCS_BoneSpace
    translationMode=BMM_Additive
    translationSpace=BCS_ParentBoneSpace
    alphaBoolBlend {blendInTime=0.5, blendOutTime=0.2}
    boneToModify {boneName=base_coat_side_rt}
    rotation=<rot -5.1215 -100.0313 -24.9645>
    translation=<-0.046 0.236 -0.311>
   }
   n42 = "Transform (Modify) Bone - Bone: offset_coat_side_lf"(ComponentPose=@n41, bAlphaBoolEnabled=True, Translation="-0.046,0.236,-0.311", Scale="1,1,1") {
    alphaInputType=Bool
    rotationMode=BMM_Additive
    rotationSpace=BCS_BoneSpace
    translationMode=BMM_Additive
    translationSpace=BCS_ParentBoneSpace
    alphaBoolBlend {blendInTime=0.5, blendOutTime=0.2}
    boneToModify {boneName=offset_coat_side_lf}
    rotation=<rot -1.7398 -29.973 -4.897>
    translation=<-0.046 0.236 -0.311>
   }
   n43 = "Transform (Modify) Bone - Bone: base_coat_front_side_rt"!(ComponentPose=@n42, bAlphaBoolEnabled:bool=True, Translation="0,0,0", Scale="1,1,1") -> (Pose) {
    alphaInputType=Bool
    rotationMode=BMM_Additive
    rotationSpace=BCS_BoneSpace
    translationMode=BMM_Additive
    translationSpace=BCS_BoneSpace
    alphaBoolBlend {blendInTime=0.5, blendOutTime=0.2}
    boneToModify {boneName=base_coat_front_side_rt}
    rotation=<rot 25.8766 -38.7451 -19.2135>
   }
   n44 = "Transform (Modify) Bone - Bone: offset_coat_front_rt"(ComponentPose=@n43, bAlphaBoolEnabled=True, Translation="-1.568,4.988,0.989", Scale="1,1,1") {
    alphaInputType=Bool
    rotationMode=BMM_Additive
    rotationSpace=BCS_BoneSpace
    translationMode=BMM_Additive
    translationSpace=BCS_BoneSpace
    alphaBoolBlend {blendInTime=0.5, blendOutTime=0.2}
    boneToModify {boneName=offset_coat_front_rt}
    rotation=<rot -21.3432 -44.8118 -13.5466>
    translation=<-1.5678 4.9882 0.9891>
   }
   n45 = "Transform (Modify) Bone - Bone: coat_corr_front_01_rt"(ComponentPose=@n44, bAlphaBoolEnabled=True, Translation="-1.390,0.123,0.057", Scale="1,1,1") {
    alphaInputType=Bool
    rotationMode=BMM_Additive
    rotationSpace=BCS_BoneSpace
    translationMode=BMM_Additive
    translationSpace=BCS_BoneSpace
    alphaBoolBlend {blendInTime=0.5, blendOutTime=0.2}
    boneToModify {boneName=coat_corr_front_01_rt}
    rotation=<rot 30.6891 -93.6381 -177.2921>
    translation=<-1.39 0.123 0.057>
   }
   n46 = "Transform (Modify) Bone - Bone: offset_coat_front_lf"(ComponentPose=@n45, bAlphaBoolEnabled=True, Translation="4.205,-0.689,0.009", Scale="1,1,1") {
    alphaInputType=Bool
    rotationMode=BMM_Additive
    rotationSpace=BCS_BoneSpace
    translationMode=BMM_Additive
    translationSpace=BCS_BoneSpace
    alphaBoolBlend {blendInTime=0.5, blendOutTime=0.2}
    boneToModify {boneName=offset_coat_front_lf}
    rotation=<rot 4.9141 -24.7281 0.051>
    translation=<4.205 -0.689 0.009>
   }
   n47 = "Transform (Modify) Bone - Bone: offset_coat_front_side_lf"(ComponentPose=@n46, bAlphaBoolEnabled=True, Translation="-0.000,-5.010,0.000", Scale="1,1,1") {
    alphaInputType=Bool
    rotationMode=BMM_Additive
    rotationSpace=BCS_BoneSpace
    translationMode=BMM_Additive
    translationSpace=BCS_BoneSpace
    alphaBoolBlend {blendInTime=0.5, blendOutTime=0.2}
    boneToModify {boneName=offset_coat_front_side_lf}
    rotation=<rot -0.0587 5.0659 -0.064>
    translation=<0 -5.0098 0>
   }
   n48 = "Transform (Modify) Bone - Bone: coat_corr_front_01_lf"(ComponentPose=@n47, bAlphaBoolEnabled=True, Translation="4.379,-1.601,-0.036", Scale="1,1,1") {
    alphaInputType=Bool
    rotationMode=BMM_Additive
    rotationSpace=BCS_BoneSpace
    translationMode=BMM_Additive
    translationSpace=BCS_BoneSpace
    alphaBoolBlend {blendInTime=0.5, blendOutTime=0.2}
    boneToModify {boneName=coat_corr_front_01_lf}
    rotation=<rot 4.962 -19.6152 -2.6973>
    translation=<4.379 -1.601 -0.036>
   }
   Save cached pose 'PreDynamics_CACHE'(Pose=Component To Local(ComponentPose=@n48)) {cacheName=PreDynamics_CACHE}
   n49 = RigidBody!(
    ComponentPose=Local To Component(LocalPose=Use cached pose 'PreDynamics_CACHE'())
    bAlphaBoolEnabled:bool
    OverrideWorldGravity:Vector="0,0,0"
    ExternalForce
   ) -> (Pose) {
    alphaInputType=Bool
    bClampLinearTranslationLimitToRefPose=true
    bOverrideWorldGravity=true
    overridePhysicsAsset=&/M_MED_RampCanal_Coat_Physics
    simulationSpace=BaseBoneSpace
    baseBoneRef {boneName=pelvis}
    simSpaceSettings {
     dampingAlpha=0
     maxAngularAcceleration=9
     maxAngularVelocity=4
     maxLinearAcceleration=1000
     maxLinearVelocity=400
     velocityScaleZ=0.75
     worldAlpha=1
     externalLinearDragV=<1.1 1.1 1.1>
    }
   }
   n50 = Component To Local(ComponentPose=@n49)
   n51 = Get bClothEnabled()
   n52 = Use cached pose 'PreDynamics_CACHE'()
   n53 = Anim Dynamics - Bone: dyn_hood_rope_front_01_lf!(
    ComponentPose=Local To Component(LocalPose=Blend Poses by bool(BlendPose_0=@n52, BlendPose_1=@n50, bActiveValue=@n51) {blendType=Linear})
    bAlphaBoolEnabled:bool=True
    GravityOverride:Vector="0,0,-980"
    ComponentLinearAccScale:Vector="1,1,1"
    ComponentLinearVelScale:Vector="0,0,0"
    ExternalForce="0,0,0"
   ) -> (Pose) {
    alphaInputType=Bool
    angularDampingOverride=1
    angularSpringConstant=11000
    bAngularSpring=true
    bOverrideAngularDamping=true
    bUseGravityOverride=true
    gravityScale=0.35
    simulationSpace=RootRelative
    alphaBoolBlend {blendInTime=0.5, blendOption=QuadraticInOut, blendOutTime=1}
    boundBone {boneName=dyn_hood_rope_front_01_lf}
    componentAppliedLinearAccClamp=<3000 3000 3000>
    componentLinearAccScale=<1 1 1>
    componentLinearVelScale=<1 1 1>
    physicsBodyDefinitions [{
      boundBone {boneName=dyn_hood_rope_front_01_lf}
      boxExtents=<20 30 30>
      constraintSetup {angularLimitsMax=<0 2 0>, angularLimitsMin=<0 -2 -5>, angularTarget=<1 0 0>}
      localJointOffset=<4 0 0>
     }]
    simSpaceSettings {maxAngularAcceleration=35, maxAngularVelocity=7, simSpaceAngularAlpha=1}
   }
   n54 = Anim Dynamics - Bone: dyn_hood_rope_front_02_lf!(
    ComponentPose=@n53
    bAlphaBoolEnabled:bool=True
    GravityOverride:Vector="0,0,-980"
    ComponentLinearAccScale:Vector="1,1,1"
    ComponentLinearVelScale:Vector="0,0,0"
    ExternalForce="0,0,0"
   ) -> (Pose) {
    alphaInputType=Bool
    angularDampingOverride=1
    angularSpringConstant=11000
    bAngularSpring=true
    bOverrideAngularDamping=true
    bUseGravityOverride=true
    gravityScale=0.35
    simulationSpace=RootRelative
    alphaBoolBlend {blendInTime=0.5, blendOption=QuadraticInOut, blendOutTime=1}
    boundBone {boneName=dyn_hood_rope_front_02_lf}
    componentAppliedLinearAccClamp=<3000 3000 3000>
    componentLinearAccScale=<1 1 1>
    componentLinearVelScale=<1 1 1>
    physicsBodyDefinitions [{
      boundBone {boneName=dyn_hood_rope_front_02_lf}
      boxExtents=<20 30 30>
      constraintSetup {angularLimitsMax=<0 5 0>, angularLimitsMin=<0 -5 -8>, angularTarget=<1 0 0>}
      localJointOffset=<4 0 0>
     }]
    simSpaceSettings {maxAngularAcceleration=35, maxAngularVelocity=7, simSpaceAngularAlpha=1}
   }
   n55 = Anim Dynamics - Bone: dyn_hood_rope_front_03_lf!(
    ComponentPose=@n54
    bAlphaBoolEnabled:bool=True
    GravityOverride:Vector="0,0,-980"
    ComponentLinearAccScale:Vector="1,1,1"
    ComponentLinearVelScale:Vector="0,0,0"
    ExternalForce="0,0,0"
   ) -> (Pose) {
    alphaInputType=Bool
    angularDampingOverride=1
    angularSpringConstant=11000
    bAngularSpring=true
    bOverrideAngularDamping=true
    bUseGravityOverride=true
    gravityScale=0.35
    simulationSpace=RootRelative
    alphaBoolBlend {blendInTime=0.5, blendOption=QuadraticInOut, blendOutTime=1}
    boundBone {boneName=dyn_hood_rope_front_03_lf}
    componentAppliedLinearAccClamp=<3000 3000 3000>
    componentLinearAccScale=<1 1 1>
    componentLinearVelScale=<1 1 1>
    physicsBodyDefinitions [{
      collisionType=CustomSphere
      sphereCollisionRadius=1.1
      boundBone {boneName=dyn_hood_rope_front_03_lf}
      boxExtents=<20 30 30>
      constraintSetup {angularLimitsMax=<0 5 -5>, angularLimitsMin=<0 -5 -11>, angularTarget=<1 0 0>}
      localJointOffset=<4 0 0>
     }]
    planarLimits [{drivingBone {boneName=spine_04}, planeTransform {location=<0 12 0>, rotation=<rot 0 0 90>}}]
    simSpaceSettings {maxAngularAcceleration=35, maxAngularVelocity=7, simSpaceAngularAlpha=1}
   }
   n56 = Anim Dynamics - Bone: dyn_hood_rope_front_01_rt!(
    ComponentPose=@n55
    bAlphaBoolEnabled:bool=True
    GravityOverride:Vector="0,0,-600"
    ComponentLinearAccScale:Vector="1,1,1"
    ComponentLinearVelScale:Vector="0,0,0"
    ExternalForce="0,0,0"
   ) -> (Pose) {
    alphaInputType=Bool
    angularDampingOverride=1
    angularSpringConstant=11000
    bAngularSpring=true
    bOverrideAngularDamping=true
    bUseGravityOverride=true
    gravityScale=0.35
    simulationSpace=RootRelative
    alphaBoolBlend {blendInTime=0.5, blendOption=QuadraticInOut, blendOutTime=1}
    boundBone {boneName=dyn_hood_rope_front_01_rt}
    componentAppliedLinearAccClamp=<3000 3000 3000>
    componentLinearAccScale=<1 1 1>
    componentLinearVelScale=<1 1 1>
    physicsBodyDefinitions [{
      boundBone {boneName=dyn_hood_rope_front_01_rt}
      boxExtents=<20 30 30>
      constraintSetup {angularLimitsMax=<0 5 0>, angularLimitsMin=<0 -5 -5>, angularTarget=<1 0 0>}
      localJointOffset=<4 0 0>
     }]
    simSpaceSettings {maxAngularAcceleration=35, maxAngularVelocity=7, simSpaceAngularAlpha=1}
   }
   n57 = Anim Dynamics - Bone: dyn_hood_rope_front_02_rt!(
    ComponentPose=@n56
    bAlphaBoolEnabled:bool=True
    GravityOverride:Vector="0,0,-600"
    ComponentLinearAccScale:Vector="1,1,1"
    ComponentLinearVelScale:Vector="0,0,0"
    ExternalForce="0,0,0"
   ) -> (Pose) {
    alphaInputType=Bool
    angularDampingOverride=1
    angularSpringConstant=11000
    bAngularSpring=true
    bOverrideAngularDamping=true
    bUseGravityOverride=true
    gravityScale=0.35
    simulationSpace=RootRelative
    alphaBoolBlend {blendInTime=0.5, blendOption=QuadraticInOut, blendOutTime=1}
    boundBone {boneName=dyn_hood_rope_front_02_rt}
    componentAppliedLinearAccClamp=<3000 3000 3000>
    componentLinearAccScale=<1 1 1>
    componentLinearVelScale=<1 1 1>
    physicsBodyDefinitions [{
      boundBone {boneName=dyn_hood_rope_front_02_rt}
      boxExtents=<20 30 30>
      constraintSetup {angularLimitsMax=<0 5 -4>, angularLimitsMin=<0 -5 -8>, angularTarget=<1 0 0>}
      localJointOffset=<4 0 0>
     }]
    simSpaceSettings {maxAngularAcceleration=35, maxAngularVelocity=7, simSpaceAngularAlpha=1}
   }
   n58 = Anim Dynamics - Bone: dyn_hood_rope_front_03_rt!(
    ComponentPose=@n57
    bAlphaBoolEnabled:bool=True
    GravityOverride:Vector="0,0,-600"
    ComponentLinearAccScale:Vector="1,1,1"
    ComponentLinearVelScale:Vector="0,0,0"
    ExternalForce="0,0,0"
   ) -> (Pose) {
    alphaInputType=Bool
    angularDampingOverride=1
    angularSpringConstant=11000
    bAngularSpring=true
    bOverrideAngularDamping=true
    bUseGravityOverride=true
    gravityScale=0.35
    simulationSpace=RootRelative
    alphaBoolBlend {blendInTime=0.5, blendOption=QuadraticInOut, blendOutTime=1}
    boundBone {boneName=dyn_hood_rope_front_03_rt}
    componentAppliedLinearAccClamp=<3000 3000 3000>
    componentLinearAccScale=<1 1 1>
    componentLinearVelScale=<1 1 1>
    physicsBodyDefinitions [{
      collisionType=CustomSphere
      sphereCollisionRadius=1.1
      boundBone {boneName=dyn_hood_rope_front_03_rt}
      boxExtents=<20 30 30>
      constraintSetup {angularLimitsMax=<0 5 -4>, angularLimitsMin=<0 -5 -11>, angularTarget=<1 0 0>}
      localJointOffset=<4 0 0>
     }]
    planarLimits [{drivingBone {boneName=spine_04}, planeTransform {location=<0 12 0>, rotation=<rot 0 0 90>}}]
    simSpaceSettings {maxAngularAcceleration=35, maxAngularVelocity=7, simSpaceAngularAlpha=1}
   }
   n59 = Anim Dynamics - Bone: dyn_hood_rope_front_04_rt!(
    ComponentPose=@n58
    bAlphaBoolEnabled:bool=True
    GravityOverride:Vector="0,0,-600"
    ComponentLinearAccScale:Vector="1,1,1"
    ComponentLinearVelScale:Vector="0,0,0"
    ExternalForce="0,0,0"
   ) -> (Pose) {
    alphaInputType=Bool
    angularDampingOverride=1
    angularSpringConstant=11000
    bAngularSpring=true
    bOverrideAngularDamping=true
    bUseGravityOverride=true
    gravityScale=0.35
    simulationSpace=RootRelative
    alphaBoolBlend {blendInTime=0.5, blendOption=QuadraticInOut, blendOutTime=1}
    boundBone {boneName=dyn_hood_rope_front_04_rt}
    componentAppliedLinearAccClamp=<3000 3000 3000>
    componentLinearAccScale=<1 1 1>
    componentLinearVelScale=<1 1 1>
    physicsBodyDefinitions [{
      collisionType=CustomSphere
      sphereCollisionRadius=1.1
      boundBone {boneName=dyn_hood_rope_front_04_rt}
      boxExtents=<20 30 30>
      constraintSetup {angularLimitsMax=<0 5 -4>, angularLimitsMin=<0 -5 -15>, angularTarget=<1 0 0>}
      localJointOffset=<4 0 0>
     }]
    planarLimits [{drivingBone {boneName=spine_04}, planeTransform {location=<0 12.5 0>, rotation=<rot 0 0 90>}}]
    simSpaceSettings {maxAngularAcceleration=35, maxAngularVelocity=7, simSpaceAngularAlpha=1}
   }
   n60 = Anim Dynamics - Bone: dyn_hood_back_01_cr_lower!(
    ComponentPose=@n59
    bAlphaBoolEnabled:bool=True
    GravityOverride:Vector="0,0,0"
    ComponentLinearAccScale:Vector="1,1,1"
    ComponentLinearVelScale:Vector="0,0,0"
    ExternalForce="0,0,0"
   ) -> (Pose) {
    alphaInputType=Bool
    angularDampingOverride=1
    angularSpringConstant=11000
    bAngularSpring=true
    bOverrideAngularDamping=true
    bUseGravityOverride=true
    gravityScale=0.35
    simulationSpace=RootRelative
    alphaBoolBlend {blendInTime=0.5, blendOption=QuadraticInOut, blendOutTime=1}
    boundBone {boneName=dyn_hood_back_01_cr_lower}
    componentAppliedLinearAccClamp=<3000 3000 3000>
    componentLinearAccScale=<1 1 1>
    componentLinearVelScale=<1 1 1>
    physicsBodyDefinitions [{
      boundBone {boneName=dyn_hood_back_01_cr_lower}
      boxExtents=<30 30 30>
      constraintSetup {angularLimitsMax=<1 15 1>, angularLimitsMin=<-1 -15 -90>, angularTarget=<1 0 0>}
      localJointOffset=<9 0 0>
     }]
    simSpaceSettings {maxAngularAcceleration=35, maxAngularVelocity=7, simSpaceAngularAlpha=1}
   }
   n61 = Anim Dynamics - Bone: offset_hood_back_cr_top!(
    ComponentPose=@n60
    bAlphaBoolEnabled:bool=True
    GravityOverride:Vector="0,0,0"
    ComponentLinearAccScale:Vector="1,1,1"
    ComponentLinearVelScale:Vector="0,0,0"
    ExternalForce="0,0,0"
   ) -> (Pose) {
    alphaInputType=Bool
    angularDampingOverride=1
    angularSpringConstant=11000
    bAngularSpring=true
    bOverrideAngularDamping=true
    bUseGravityOverride=true
    gravityScale=0.35
    simulationSpace=RootRelative
    alphaBoolBlend {blendInTime=0.5, blendOption=QuadraticInOut, blendOutTime=1}
    boundBone {boneName=offset_hood_back_cr_top}
    componentAppliedLinearAccClamp=<3000 3000 3000>
    componentLinearAccScale=<1 1 1>
    componentLinearVelScale=<1 1 1>
    physicsBodyDefinitions [{
      boundBone {boneName=offset_hood_back_cr_top}
      boxExtents=<20 30 30>
      constraintSetup {angularLimitsMax=<1 15 30>, angularLimitsMin=<-1 -15 -10>, angularTarget=<1 0 0>}
      localJointOffset=<4 0 0>
     }]
    simSpaceSettings {maxAngularAcceleration=35, maxAngularVelocity=7, simSpaceAngularAlpha=1}
   }
   n62 = Anim Dynamics - Bone: dyn_zipper_front_lf!(
    ComponentPose=@n61
    bAlphaBoolEnabled:bool=True
    GravityOverride:Vector="0,0,-980"
    ComponentLinearAccScale:Vector="1,1,1"
    ComponentLinearVelScale:Vector="0,0,0"
    ExternalForce="0,0,0"
   ) -> (Pose) {
    alphaInputType=Bool
    angularDampingOverride=1
    angularSpringConstant=11000
    bAngularSpring=true
    bOverrideAngularDamping=true
    bUseGravityOverride=true
    gravityScale=0.35
    simulationSpace=RootRelative
    alphaBoolBlend {blendInTime=0.5, blendOption=QuadraticInOut, blendOutTime=1}
    boundBone {boneName=dyn_zipper_front_lf}
    componentAppliedLinearAccClamp=<3000 3000 3000>
    componentLinearAccScale=<1 1 1>
    componentLinearVelScale=<1 1 1>
    physicsBodyDefinitions [{
      collisionType=CustomSphere
      sphereCollisionRadius=1
      boundBone {boneName=dyn_zipper_front_lf}
      boxExtents=<20 30 30>
      constraintSetup {angularLimitsMax=<1 5 0>, angularLimitsMin=<-1 -5 -20>, angularTarget=<1 0 0>}
      localJointOffset=<4 0 0>
     }]
    planarLimits [{drivingBone {boneName=base_zipper_front_lf}, planeTransform {location=<0 0 0>, rotation=<rot 0 0 -90>}}]
    simSpaceSettings {maxAngularAcceleration=35, maxAngularVelocity=7, simSpaceAngularAlpha=1}
   }
   n63 = Anim Dynamics - Bone: dyn_skirt_rope_back_01_lf!(
    ComponentPose=@n62
    bAlphaBoolEnabled:bool=True
    GravityOverride:Vector="0,0,-980"
    ComponentLinearAccScale:Vector="1,1,1"
    ComponentLinearVelScale:Vector="0,0,0"
    ExternalForce="0,0,0"
   ) -> (Pose) {
    alphaInputType=Bool
    angularDampingOverride=1
    angularSpringConstant=11000
    bAngularSpring=true
    bOverrideAngularDamping=true
    bUseGravityOverride=true
    gravityScale=0.35
    simulationSpace=RootRelative
    alphaBoolBlend {blendInTime=0.5, blendOption=QuadraticInOut, blendOutTime=1}
    boundBone {boneName=dyn_skirt_rope_back_01_lf}
    componentAppliedLinearAccClamp=<3000 3000 3000>
    componentLinearAccScale=<1 1 1>
    componentLinearVelScale=<1 1 1>
    physicsBodyDefinitions [{
      boundBone {boneName=dyn_skirt_rope_back_01_lf}
      boxExtents=<20 30 30>
      constraintSetup {angularLimitsMax=<0 5 0>, angularLimitsMin=<0 -5 -5>, angularTarget=<1 0 0>}
      localJointOffset=<4 0 0>
     }]
    simSpaceSettings {maxAngularAcceleration=35, maxAngularVelocity=7, simSpaceAngularAlpha=1}
   }
   n64 = Anim Dynamics - Bone: dyn_skirt_rope_back_02_lf!(
    ComponentPose=@n63
    bAlphaBoolEnabled:bool=True
    GravityOverride:Vector="0,0,-980"
    ComponentLinearAccScale:Vector="1,1,1"
    ComponentLinearVelScale:Vector="0,0,0"
    ExternalForce="0,0,0"
   ) -> (Pose) {
    alphaInputType=Bool
    angularDampingOverride=1
    angularSpringConstant=11000
    bAngularSpring=true
    bOverrideAngularDamping=true
    bUseGravityOverride=true
    gravityScale=0.35
    simulationSpace=RootRelative
    alphaBoolBlend {blendInTime=0.5, blendOption=QuadraticInOut, blendOutTime=1}
    boundBone {boneName=dyn_skirt_rope_back_02_lf}
    componentAppliedLinearAccClamp=<3000 3000 3000>
    componentLinearAccScale=<1 1 1>
    componentLinearVelScale=<1 1 1>
    physicsBodyDefinitions [{
      boundBone {boneName=dyn_skirt_rope_back_02_lf}
      boxExtents=<20 30 30>
      constraintSetup {angularLimitsMax=<0 8 0>, angularLimitsMin=<0 -8 -8>, angularTarget=<1 0 0>}
      localJointOffset=<4 0 0>
     }]
    simSpaceSettings {maxAngularAcceleration=35, maxAngularVelocity=7, simSpaceAngularAlpha=1}
   }
   n65 = Anim Dynamics - Bone: dyn_skirt_rope_back_03_lf!(
    ComponentPose=@n64
    bAlphaBoolEnabled:bool=True
    GravityOverride:Vector="0,0,-980"
    ComponentLinearAccScale:Vector="1,1,1"
    ComponentLinearVelScale:Vector="0,0,0"
    ExternalForce="0,0,0"
   ) -> (Pose) {
    alphaInputType=Bool
    angularDampingOverride=1
    angularSpringConstant=11000
    bAngularSpring=true
    bOverrideAngularDamping=true
    bUseGravityOverride=true
    gravityScale=0.35
    simulationSpace=RootRelative
    alphaBoolBlend {blendInTime=0.5, blendOption=QuadraticInOut, blendOutTime=1}
    boundBone {boneName=dyn_skirt_rope_back_03_lf}
    componentAppliedLinearAccClamp=<3000 3000 3000>
    componentLinearAccScale=<1 1 1>
    componentLinearVelScale=<1 1 1>
    physicsBodyDefinitions [{
      boundBone {boneName=dyn_skirt_rope_back_03_lf}
      boxExtents=<20 30 30>
      constraintSetup {angularLimitsMax=<0 11 0>, angularLimitsMin=<0 -11 -11>, angularTarget=<1 0 0>}
      localJointOffset=<4 0 0>
     }]
    simSpaceSettings {maxAngularAcceleration=35, maxAngularVelocity=7, simSpaceAngularAlpha=1}
   }
   n66 = Anim Dynamics - Bone: dyn_skirt_rope_back_01_rt!(
    ComponentPose=@n65
    bAlphaBoolEnabled:bool=True
    GravityOverride:Vector="0,0,-980"
    ComponentLinearAccScale:Vector="1,1,1"
    ComponentLinearVelScale:Vector="0,0,0"
    ExternalForce="0,0,0"
   ) -> (Pose) {
    alphaInputType=Bool
    angularDampingOverride=1
    angularSpringConstant=11000
    bAngularSpring=true
    bOverrideAngularDamping=true
    bUseGravityOverride=true
    gravityScale=0.35
    simulationSpace=RootRelative
    alphaBoolBlend {blendInTime=0.5, blendOption=QuadraticInOut, blendOutTime=1}
    boundBone {boneName=dyn_skirt_rope_back_01_rt}
    componentAppliedLinearAccClamp=<3000 3000 3000>
    componentLinearAccScale=<1 1 1>
    componentLinearVelScale=<1 1 1>
    physicsBodyDefinitions [{
      boundBone {boneName=dyn_skirt_rope_back_01_rt}
      boxExtents=<20 30 30>
      constraintSetup {angularLimitsMax=<0 5 0>, angularLimitsMin=<0 -5 -5>, angularTarget=<1 0 0>}
      localJointOffset=<4 0 0>
     }]
    simSpaceSettings {maxAngularAcceleration=35, maxAngularVelocity=7, simSpaceAngularAlpha=1}
   }
   n67 = Anim Dynamics - Bone: dyn_skirt_rope_back_02_rt!(
    ComponentPose=@n66
    bAlphaBoolEnabled:bool=True
    GravityOverride:Vector="0,0,-980"
    ComponentLinearAccScale:Vector="1,1,1"
    ComponentLinearVelScale:Vector="0,0,0"
    ExternalForce="0,0,0"
   ) -> (Pose) {
    alphaInputType=Bool
    angularDampingOverride=1
    angularSpringConstant=11000
    bAngularSpring=true
    bOverrideAngularDamping=true
    bUseGravityOverride=true
    gravityScale=0.35
    simulationSpace=RootRelative
    alphaBoolBlend {blendInTime=0.5, blendOption=QuadraticInOut, blendOutTime=1}
    boundBone {boneName=dyn_skirt_rope_back_02_rt}
    componentAppliedLinearAccClamp=<3000 3000 3000>
    componentLinearAccScale=<1 1 1>
    componentLinearVelScale=<1 1 1>
    physicsBodyDefinitions [{
      boundBone {boneName=dyn_skirt_rope_back_02_rt}
      boxExtents=<20 30 30>
      constraintSetup {angularLimitsMax=<0 8 0>, angularLimitsMin=<0 -8 -8>, angularTarget=<1 0 0>}
      localJointOffset=<4 0 0>
     }]
    simSpaceSettings {maxAngularAcceleration=35, maxAngularVelocity=7, simSpaceAngularAlpha=1}
   }
   n68 = Anim Dynamics - Bone: dyn_skirt_rope_back_03_rt!(
    ComponentPose=@n67
    bAlphaBoolEnabled:bool=True
    GravityOverride:Vector="0,0,-980"
    ComponentLinearAccScale:Vector="1,1,1"
    ComponentLinearVelScale:Vector="0,0,0"
    ExternalForce="0,0,0"
   ) -> (Pose) {
    alphaInputType=Bool
    angularDampingOverride=1
    angularSpringConstant=11000
    bAngularSpring=true
    bOverrideAngularDamping=true
    bUseGravityOverride=true
    gravityScale=0.35
    simulationSpace=RootRelative
    alphaBoolBlend {blendInTime=0.5, blendOption=QuadraticInOut, blendOutTime=1}
    boundBone {boneName=dyn_skirt_rope_back_03_rt}
    componentAppliedLinearAccClamp=<3000 3000 3000>
    componentLinearAccScale=<1 1 1>
    componentLinearVelScale=<1 1 1>
    physicsBodyDefinitions [{
      boundBone {boneName=dyn_skirt_rope_back_03_rt}
      boxExtents=<20 30 30>
      constraintSetup {angularLimitsMax=<0 11 0>, angularLimitsMin=<0 -11 -11>, angularTarget=<1 0 0>}
      localJointOffset=<4 0 0>
     }]
    simSpaceSettings {maxAngularAcceleration=35, maxAngularVelocity=7, simSpaceAngularAlpha=1}
   }
   n69 = "Constraint - base_zipper_front_lf (PPPP)"!(
    ComponentPose=@n68
    Alpha=1
    ConstraintWeights_0=0.046
    ConstraintWeights_1:float=0.236
    ConstraintWeights_2:float=0.623
    ConstraintWeights_3:float=0.095
   ) -> (Pose) {
    boneToModify {boneName=base_zipper_front_lf}
    constraintSetup [
     {targetBone {boneName=dyn_coat_front_02_lf}}
     {targetBone {boneName=dyn_coat_front_03_lf}}
     {targetBone {boneName=dyn_coat_front_04_lf}}
     {targetBone {boneName=dyn_coat_front_05_lf}}
    ]
    constraintWeights [1,1,1,1]
   }
   Save cached pose 'Dynamics_CACHE'(Pose=Component To Local(ComponentPose=@n69)) {cacheName=Dynamics_CACHE}
  ]}
 EventGraph {nodes [
   n1 = Event BlueprintUpdateAnimation()
   > Set DeltaTimeLocal(DeltaTimeLocal:double=@n1.DeltaTimeX) -> (Output_Get:double)
   > n2 = Sequence!(execute) -> (then_0, then_1, then_2:Exec, then_3:Exec)
   n3 = Get bIsMoving2D()
   Set CrouchIDLE(execute=@n2.then_2, CrouchIDLE:bool=AND Boolean(A=Get bIsCrouching(), B=NOT Boolean(A=@n3))) -> (Output_Get:bool)
   n4 = Get bIsSkyDiving()
   > Set IsJumpingCustom(IsJumpingCustom:bool=AND Boolean(A=Get bIsJumping(), B=NOT Boolean(A=@n4))) -> (Output_Get:bool)
   n5 = Get bIsDBNO()
   n6 = Get bIsCrouching()
   n7 = Get bIsSwimSprinting() -> (bIsSwimSprinting:bool)
   > Set IsGroundMovement(IsGroundMovement:bool=AND Boolean!(A=Get bIsSprinting(), B=NOT Boolean(A=@n7), C:bool=NOT Boolean(A=@n6), D:bool=NOT Boolean(A=@n5)) -> (ReturnValue)) -> (Output_Get:bool)
   > Set IsCrouchWalking(IsCrouchWalking:bool=AND Boolean!(A=Get bIsCrouching(), B=Get bIsMoving2D()) -> (ReturnValue)) -> (Output_Get:bool)
   > Set IsMeleeNotCrouching(IsMeleeNotCrouching:bool=AND Boolean(A=Get bIsCrouching(), B=Get bIsPlayingMeleeAnim())) -> (Output_Get:bool)
   n8 = GetOwnerAnimationInterface!(self:Custom Character Part Anim Instance Object Reference) -> (ReturnValue:Fort Player Animation Interface Object Reference)
   Is Valid(exec=@n2.then_1, InputObject=@n8)
   > n9 = Branch(Condition=Break Fort Anim Input Zipline!(FortAnimInput_Zipline:"Fort Anim Input Zipline Structure (by ref)"=GetZiplineInput!(self:Fort Player Animation Interface Object Reference=@n8) -> (ReturnValue:"Fort Anim Input Zipline Structure (by ref)")) -> (
     bIsZiplining:bool=~
     bIsOnAscender:bool=~
     bShouldPlayPivotTransition:bool=~
     LeanYaw:float=~
     PivotBlendDelayRemaining:float=~
     PivotBlendDelay:float=~
     PivotCardinalDirection:EFortCardinalDirection Enum
     WorldVelocityLastTick:Vector=~
    ) {is=title K2Node_BreakStruct})
   > Set Componen Linear Vel Scale(Componen Linear Vel Scale:Vector="0, 0, 0") -> (Output_Get:Vector)
   Set Componen Linear Vel Scale(execute=@n9.else, Componen Linear Vel Scale="0.3,0.3,0.3")
   n10 = Branch(execute=@n2.then_0, Condition=Get bClothEnabled())
   > n11 = Sequence!(execute) -> (then_0, then_1)
   "Cloth Controller Function (Parameter Asset) (Nativized)"!(
    execute=@n11.then_1
    CharacterPartAnimInstance:Custom Character Part Anim Instance Object Reference=Self-Reference()
    ClothingAssetName:Name=M_MED_RampCanal_Clothing
    ClothCoreSettings:"Cloth Core Settings Structure (by ref)"=Get Cloth Core Settings() -> (Cloth Core Settings:Cloth Core Settings Structure)
    WindGustLocalSettings:"Wind Gust Settings Structure (by ref)"=Get Wind Gust Local Settings() -> (Wind Gust Local Settings:Wind Gust Settings Structure)
    ClothUpAxis:Vector="1,0,0"
    LinVelScaleFwdBwdClamp:float=0
    LinVelScaleUpDwnClamp:float=0.7
    DeltaTime:float=Get DeltaTimeLocal() -> (DeltaTimeLocal:double)
    GameTimeInSeconds:float=GetGameTimeInSeconds()
    StateLogicLayerName:Name=none
    Parameters:Fort Character Dynamics Parameters Object Reference=&/M_MED_RampCanal_Parameters
   ) -> (
    then
    OutCurrentState:Name=None
    OutClothParameters:Cloth Parameters Structure
    OutClothCoreSettings:Cloth Core Settings Structure
    OutWindGustSettings:Wind Gust Parameters Structure
    OutWindGustLocalSettings:Wind Gust Settings Structure
    OutWindVelocity:Vector
    OutLinearVelVector:Vector
   )
   n12 = Branch(execute=@n11.then_0, Condition=Get bIsPlayingEmote())
   n13 = GetOwnerSkeletalMeshComponent()
   > n14 = Branch(Condition=Montage_IsActive!(
     self:Anim Instance Object Reference=GetAnimInstance(self=@n13)
     Montage:Anim Montage Object Reference=/BRCosmetics/Animation/Game/MainPlayer/Emotes/Kitty_Cat/Emote_Kitty_Cat_CMM_M
    ) -> (ReturnValue:bool))
   > Set Cloth Core Settings(Cloth Core Settings:Cloth Core Settings Structure=Get Cloth Core Settings_Emote_CatFlip() -> (Cloth Core Settings_Emote_CatFlip:Cloth Core Settings Structure)) -> (Output_Get:Cloth Core Settings Structure)
   Set Cloth Core Settings(execute=@n14.else, Cloth Core Settings=Get Cloth Core Settings_Emotes() -> (Cloth Core Settings_Emotes:Cloth Core Settings Structure))
   Set Cloth Core Settings(execute=@n12.else, Cloth Core Settings=Get Cloth Core Settings_Normal() -> (Cloth Core Settings_Normal:Cloth Core Settings Structure))
   n15 = "Gravity Override State Function (Parameter Asset) (Nativized)"(
    execute=@n10.else
    CharacterPartAnimInstance=Self-Reference()
    GameTimeInSeconds=GetGameTimeInSeconds()
    Parameters=&M_MED_RampCanal_Parameters
   )
   > Set Gravity Override(Gravity Override=@n15)
   Event OnCharacterCustomizationLoaded()
   n16 = CheckForBackbling()
   > Set IsWearingBackpackOrCape(IsWearingBackpackOrCape:bool=NOT Boolean(A=OR Boolean!(A=@n16, B=@n16.bIsBackblingACape, C:bool=@n16.bIsBackblingLong) -> (ReturnValue))) -> (Output_Get:bool)
  ]}
]
variables [
 ChaosClothingInteractor {type=ChaosClothingInteractor, default=null}
 Cloth Core Settings {
  type=ClothCoreSettings
  default {
   collisionThickness=0
   dampingCoefficient=0.05
   frictionCoefficient=0.3
   localDampingCoefficient=0
   numberIterations=2
   numberSubsteps=1
   selfCollisionThickness=0
   animDriveDamping=<0 1>
   animDriveStiffness=<0 1>
   areaStiffness=<1 1>
   bendingStiffness=<1 1>
   edgeStiffness=<1 1>
   tetherStiffness=<1 1>
  }
 }
 Cloth Core Settings_Emote_CatFlip {
  type=ClothCoreSettings
  default {
   collisionThickness=0
   dampingCoefficient=0.05
   frictionCoefficient=0.3
   localDampingCoefficient=0
   numberIterations=2
   numberSubsteps=1
   selfCollisionThickness=0
   animDriveDamping=<0 1>
   animDriveStiffness=<0.38 1>
   areaStiffness=<1 1>
   bendingStiffness=<1 1>
   edgeStiffness=<1 1>
   tetherStiffness=<1 1>
  }
 }
 Cloth Core Settings_Emotes {
  type=ClothCoreSettings
  default {
   collisionThickness=0
   dampingCoefficient=0.05
   frictionCoefficient=0.3
   localDampingCoefficient=0
   numberIterations=2
   numberSubsteps=1
   selfCollisionThickness=0
   animDriveDamping=<0 1>
   animDriveStiffness=<0.25 1>
   areaStiffness=<1 1>
   bendingStiffness=<1 1>
   edgeStiffness=<1 1>
   tetherStiffness=<1 1>
  }
 }
 Cloth Core Settings_Normal {
  type=ClothCoreSettings
  default {
   collisionThickness=0
   dampingCoefficient=0.05
   frictionCoefficient=0.3
   localDampingCoefficient=0
   numberIterations=2
   numberSubsteps=1
   selfCollisionThickness=0
   animDriveDamping=<0 1>
   animDriveStiffness=<0 1>
   areaStiffness=<1 1>
   bendingStiffness=<1 1>
   edgeStiffness=<1 1>
   tetherStiffness=<1 1>
  }
 }
 Componen Linear Vel Scale {type=Vector, default=<0.3 0.3 0.3>}
 CrouchIDLE {type=boolean}
 DeltaTimeLocal {type=number}
 EndEmote {type=boolean}
 Gravity Override {type=Vector}
 IsCrouchWalking {type=boolean}
 IsGroundMovement {type=boolean}
 IsJumpingCustom {type=boolean}
 IsJumpingFix {type=boolean}
 IsMeleeNotCrouching {type=boolean}
 IsWearingBackpackOrCape {type=boolean}
 Wind Gust Local Settings {type=WindGustSettings, default {elapsedTime=0, finalGustStrengthLocal=0, gustDuration=0, gustStrength=0, timeBetweenGusts=0}}
]
