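# Physics asset that supplies cloth collision for the medium male RampCanal player character body mesh.
# Capsules on pelvis, spine_01 and spine_02 approximate the lower torso.
# A wide box on root sits just below the origin, its top near ground level like a floor plane.
# Leg, foot and pelvis_cloth_helper bodies carry no shapes, and no constraints link the bodies.
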PhysicsAsset /BRCosmetics/Characters/Player/Male/Medium/Bodies/M_MED_RampCanal/Meshes/M_MED_RampCanal_ClothCollision
bodies {
 calf_l []
 calf_r []
 foot_l []
 foot_r []
 pelvis [{shapeType=Capsule, length=8.5, radius=9.8, center=<-2.5 2 0>, rotation=<rot 0 -0.0006 0>}]
 pelvis_cloth_helper []
 root [{shapeType=Box, extentX=327.5, extentY=327.5, extentZ=59.22, shapeName=root_box, center=<0 0 -29>}]
 spine_01 [{shapeType=Capsule, length=8.5, radius=9.5, center=<-2.5 2.5 0>}]
 spine_02 [{shapeType=Capsule, length=9, radius=9, center=<0.3095 2.2 0>}]
 thigh_l []
 thigh_r []
}
constraints []
