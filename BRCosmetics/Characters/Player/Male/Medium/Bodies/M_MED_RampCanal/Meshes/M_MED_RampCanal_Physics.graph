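# Physics asset for the medium male RampCanal player body cosmetic, covering the back hood and upper spine.
# Single capsules represent the hood base bone, the dynamic hood bone and spine_05, giving the hood a torso shape to collide with.
# One constraint joins the dynamic hood bone to its base with limited, asymmetric swing and very tight twist, while the spine capsule has no constraint.
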PhysicsAsset /BRCosmetics/Characters/Player/Male/Medium/Bodies/M_MED_RampCanal/Meshes/M_MED_RampCanal_Physics
bodies {
 base_hood_back_cr_lower [{shapeType=Capsule, length=0.505, radius=0.505, shapeName=base_hood_back_cr_lower_capsule, rotation=<rot 0 -90 0>}]
 dyn_hood_back_01_cr_lower [{shapeType=Capsule, length=8, radius=4, shapeName=dyn_hood_back_01_cr_lower_capsule, center=<6 -3 0>, rotation=<rot 0 0 2.5>}]
 spine_05 [{shapeType=Capsule, length=22, radius=11.2, shapeName=spine_05_capsule, center=<0 0.5 0>}]
}
constraints [{bone1Name=dyn_hood_back_01_cr_lower, bone2Name=base_hood_back_cr_lower, swing1LimitDegrees=40, swing2LimitDegrees=20, twistLimitDegrees=1}]
