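# Post-pose AnimBP for the RampCanal face accessory mesh, adding spring-driven secondary motion to twelve dyn_hair bones.
# The parent mesh's pose runs through chained, angle-limited Anim Dynamics nodes.
# A per-bone blend brings DefaultSlot onto hair bones, weighted by the front-end Frontend_Idle_Bypass curve.
# Updates also set gravity and zipline-dependent velocity-scale variables, though the dynamics pins show fixed values.
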AnimBlueprint /BRCosmetics/Characters/Player/Male/Medium/Bodies/M_MED_RampCanal/Meshes/Parts/M_MED_RampCanal_FaceAcc_AnimBP
target_skeleton=&/M_MED_RampCanal_FaceAcc_Skeleton
graphs [
 AnimGraph {nodes [
   n1 = Anim Dynamics - Bone: dyn_hair_02!(
    ComponentPose=Local To Component(LocalPose=Copy Pose From Mesh() {bUseAttachedParent=true})
    bAlphaBoolEnabled:bool=True
    GravityOverride:Vector="0,0,0"
    ComponentLinearAccScale:Vector="1,1,1"
    ComponentLinearVelScale:Vector="0,0,0"
    ExternalForce
   ) -> (Pose) {
    alphaInputType=Bool
    angularDampingOverride=1
    angularSpringConstant=11000
    bAngularSpring=true
    bOverrideAngularDamping=true
    bUseGravityOverride=true
    gravityScale=0.35
    simulationSpace=RootRelative
    alphaBoolBlend {blendInTime=0.5, blendOption=QuadraticInOut, blendOutTime=1}
    boundBone {boneName=dyn_hair_02}
    componentAppliedLinearAccClamp=<3000 3000 3000>
    componentLinearAccScale=<1 1 1>
    componentLinearVelScale=<1 1 1>
    physicsBodyDefinitions [{
      boundBone {boneName=dyn_hair_02}
      boxExtents=<15 25 25>
      constraintSetup {angularLimitsMax=<1 5 1>, angularLimitsMin=<-1 -5 -10>, angularTarget=<1 0 0>}
      localJointOffset=<4 0 0>
     }]
    simSpaceSettings {maxAngularAcceleration=35, maxAngularVelocity=7, simSpaceAngularAlpha=1}
   }
   n2 = Anim Dynamics - Bone: dyn_hair_05!(
    ComponentPose=@n1
    bAlphaBoolEnabled:bool=True
    GravityOverride:Vector="0,0,0"
    ComponentLinearAccScale:Vector="1,1,1"
    ComponentLinearVelScale:Vector="0,0,0"
    ExternalForce="0,0,0"
   ) -> (Pose) {
    alphaInputType=Bool
    angularDampingOverride=1
    angularSpringConstant=11000
    bAngularSpring=true
    bOverrideAngularDamping=true
    bUseGravityOverride=true
    gravityScale=0.35
    simulationSpace=RootRelative
    alphaBoolBlend {blendInTime=0.5, blendOption=QuadraticInOut, blendOutTime=1}
    boundBone {boneName=dyn_hair_05}
    componentAppliedLinearAccClamp=<3000 3000 3000>
    componentLinearAccScale=<1 1 1>
    componentLinearVelScale=<1 1 1>
    physicsBodyDefinitions [{
      boundBone {boneName=dyn_hair_05}
      boxExtents=<20 30 30>
      constraintSetup {angularLimitsMax=<1 5 1>, angularLimitsMin=<-1 -5 -3>, angularTarget=<1 0 0>}
      localJointOffset=<6 0 0>
     }]
    simSpaceSettings {maxAngularAcceleration=35, maxAngularVelocity=7, simSpaceAngularAlpha=1}
   }
   n3 = Anim Dynamics - Bone: dyn_hair_08!(
    ComponentPose=@n2
    bAlphaBoolEnabled:bool=True
    GravityOverride:Vector="0,0,-980"
    ComponentLinearAccScale:Vector="1,1,1"
    ComponentLinearVelScale:Vector="0,0,0"
    ExternalForce="0,0,0"
   ) -> (Pose) {
    alphaInputType=Bool
    angularDampingOverride=1
    angularSpringConstant=11000
    bAngularSpring=true
    bOverrideAngularDamping=true
    bUseGravityOverride=true
    gravityScale=0.35
    simulationSpace=RootRelative
    alphaBoolBlend {blendInTime=0.5, blendOption=QuadraticInOut, blendOutTime=1}
    boundBone {boneName=dyn_hair_08}
    componentAppliedLinearAccClamp=<3000 3000 3000>
    componentLinearAccScale=<1 1 1>
    componentLinearVelScale=<1 1 1>
    physicsBodyDefinitions [{
      boundBone {boneName=dyn_hair_08}
      boxExtents=<22 30 33>
      constraintSetup {angularLimitsMax=<1 3 1>, angularLimitsMin=<-1 -3 -4>, angularTarget=<1 0 0>}
      localJointOffset=<2 0 0>
     }]
    simSpaceSettings {maxAngularAcceleration=35, maxAngularVelocity=7, simSpaceAngularAlpha=1}
   }
   n4 = Anim Dynamics - Bone: dyn_hair_09!(
    ComponentPose=@n3
    bAlphaBoolEnabled:bool=True
    GravityOverride:Vector="0,0,0"
    ComponentLinearAccScale:Vector="1,1,1"
    ComponentLinearVelScale:Vector="0,0,0"
    ExternalForce="0,0,0"
   ) -> (Pose) {
    alphaInputType=Bool
    angularDampingOverride=1
    angularSpringConstant=11000
    bAngularSpring=true
    bOverrideAngularDamping=true
    bUseGravityOverride=true
    gravityScale=0.35
    simulationSpace=RootRelative
    alphaBoolBlend {blendInTime=0.5, blendOption=QuadraticInOut, blendOutTime=1}
    boundBone {boneName=dyn_hair_09}
    componentAppliedLinearAccClamp=<3000 3000 3000>
    componentLinearAccScale=<1 1 1>
    componentLinearVelScale=<1 1 1>
    physicsBodyDefinitions [{
      boundBone {boneName=dyn_hair_09}
      boxExtents=<15 35 35>
      constraintSetup {angularLimitsMax=<1 20 15>, angularLimitsMin=<-1 -20 -15>, angularTarget=<1 0 0>}
      localJointOffset=<5 0 0>
     }]
    simSpaceSettings {maxAngularAcceleration=35, maxAngularVelocity=7, simSpaceAngularAlpha=1}
   }
   n5 = Anim Dynamics - Bone: dyn_hair_10!(
    ComponentPose=@n4
    bAlphaBoolEnabled:bool=True
    GravityOverride:Vector="0,0,-100"
    ComponentLinearAccScale:Vector="1,1,1"
    ComponentLinearVelScale:Vector="0,0,0"
    ExternalForce="0,0,0"
   ) -> (Pose) {
    alphaInputType=Bool
    angularDampingOverride=1
    angularSpringConstant=11000
    bAngularSpring=true
    bOverrideAngularDamping=true
    bUseGravityOverride=true
    gravityScale=0.35
    simulationSpace=RootRelative
    alphaBoolBlend {blendInTime=0.5, blendOption=QuadraticInOut, blendOutTime=1}
    boundBone {boneName=dyn_hair_10}
    componentAppliedLinearAccClamp=<3000 3000 3000>
    componentLinearAccScale=<1 1 1>
    componentLinearVelScale=<1 1 1>
    physicsBodyDefinitions [{
      boundBone {boneName=dyn_hair_10}
      boxExtents=<25 25 25>
      constraintSetup {angularLimitsMax=<1 13 13>, angularLimitsMin=<-1 -13 -13>, angularTarget=<1 0 0>}
      localJointOffset=<4 0 0>
     }]
    simSpaceSettings {maxAngularAcceleration=35, maxAngularVelocity=7, simSpaceAngularAlpha=1}
   }
   n6 = Anim Dynamics - Bone: dyn_hair_13!(
    ComponentPose=@n5
    bAlphaBoolEnabled:bool=True
    GravityOverride:Vector="0,0,0"
    ComponentLinearAccScale:Vector="1,1,1"
    ComponentLinearVelScale:Vector="0,0,0"
    ExternalForce="0,0,0"
   ) -> (Pose) {
    alphaInputType=Bool
    angularDampingOverride=1
    angularSpringConstant=11000
    bAngularSpring=true
    bOverrideAngularDamping=true
    bUseGravityOverride=true
    gravityScale=0.35
    simulationSpace=RootRelative
    alphaBoolBlend {blendInTime=0.5, blendOption=QuadraticInOut, blendOutTime=1}
    boundBone {boneName=dyn_hair_13}
    componentAppliedLinearAccClamp=<3000 3000 3000>
    componentLinearAccScale=<1 1 1>
    componentLinearVelScale=<1 1 1>
    physicsBodyDefinitions [{
      boundBone {boneName=dyn_hair_13}
      boxExtents=<15 25 25>
      constraintSetup {angularLimitsMax=<1 10 5>, angularLimitsMin=<-1 -10 -5>, angularTarget=<1 0 0>}
      localJointOffset=<6 0 0>
     }]
    simSpaceSettings {maxAngularAcceleration=35, maxAngularVelocity=7, simSpaceAngularAlpha=1}
   }
   n7 = Anim Dynamics - Bone: dyn_hair_14!(
    ComponentPose=@n6
    bAlphaBoolEnabled:bool=True
    GravityOverride:Vector="0,0,0"
    ComponentLinearAccScale:Vector="1,1,1"
    ComponentLinearVelScale:Vector="0,0,0"
    ExternalForce="0,0,0"
   ) -> (Pose) {
    alphaInputType=Bool
    angularDampingOverride=1
    angularSpringConstant=11000
    bAngularSpring=true
    bOverrideAngularDamping=true
    bUseGravityOverride=true
    gravityScale=0.35
    simulationSpace=RootRelative
    alphaBoolBlend {blendInTime=0.5, blendOption=QuadraticInOut, blendOutTime=1}
    boundBone {boneName=dyn_hair_14}
    componentAppliedLinearAccClamp=<3000 3000 3000>
    componentLinearAccScale=<1 1 1>
    componentLinearVelScale=<1 1 1>
    physicsBodyDefinitions [{
      boundBone {boneName=dyn_hair_14}
      boxExtents=<22 33 33>
      constraintSetup {angularLimitsMax=<1 9 6>, angularLimitsMin=<-1 -6 -9>, angularTarget=<1 0 0>}
      localJointOffset=<3 0 0>
     }]
    simSpaceSettings {maxAngularAcceleration=35, maxAngularVelocity=7, simSpaceAngularAlpha=1}
   }
   n8 = Anim Dynamics - Bone: dyn_hair_16!(
    ComponentPose=@n7
    bAlphaBoolEnabled:bool=True
    GravityOverride:Vector="0,0,0"
    ComponentLinearAccScale:Vector="1,1,1"
    ComponentLinearVelScale:Vector="0,0,0"
    ExternalForce="0,0,0"
   ) -> (Pose) {
    alphaInputType=Bool
    angularDampingOverride=1
    angularSpringConstant=11000
    bAngularSpring=true
    bOverrideAngularDamping=true
    bUseGravityOverride=true
    gravityScale=0.35
    simulationSpace=RootRelative
    alphaBoolBlend {blendInTime=0.5, blendOption=QuadraticInOut, blendOutTime=1}
    boundBone {boneName=dyn_hair_16}
    componentAppliedLinearAccClamp=<3000 3000 3000>
    componentLinearAccScale=<1 1 1>
    componentLinearVelScale=<1 1 1>
    physicsBodyDefinitions [{
      boundBone {boneName=dyn_hair_16}
      boxExtents=<20 30 30>
      constraintSetup {angularLimitsMax=<1 18 11>, angularLimitsMin=<-1 -20 -15>, angularTarget=<1 0 0>}
      localJointOffset=<5 0 0>
     }]
    simSpaceSettings {maxAngularAcceleration=35, maxAngularVelocity=7, simSpaceAngularAlpha=1}
   }
   n9 = Anim Dynamics - Bone: dyn_hair_20!(
    ComponentPose=@n8
    bAlphaBoolEnabled:bool=True
    GravityOverride:Vector="0,0,0"
    ComponentLinearAccScale:Vector="1,1,1"
    ComponentLinearVelScale:Vector="0,0,0"
    ExternalForce="0,0,0"
   ) -> (Pose) {
    alphaInputType=Bool
    angularDampingOverride=1
    angularSpringConstant=11000
    bAngularSpring=true
    bOverrideAngularDamping=true
    bUseGravityOverride=true
    gravityScale=0.35
    simulationSpace=RootRelative
    alphaBoolBlend {blendInTime=0.5, blendOption=QuadraticInOut, blendOutTime=1}
    boundBone {boneName=dyn_hair_20}
    componentAppliedLinearAccClamp=<3000 3000 3000>
    componentLinearAccScale=<1 1 1>
    componentLinearVelScale=<1 1 1>
    physicsBodyDefinitions [{
      boundBone {boneName=dyn_hair_20}
      boxExtents=<15 25 25>
      constraintSetup {angularLimitsMax=<1 5 3>, angularLimitsMin=<-1 -3 -5>, angularTarget=<1 0 0>}
      localJointOffset=<3 0 0>
     }]
    simSpaceSettings {maxAngularAcceleration=35, maxAngularVelocity=7, simSpaceAngularAlpha=1}
   }
   n10 = Anim Dynamics - Bone: dyn_hair_24!(
    ComponentPose=@n9
    bAlphaBoolEnabled:bool=True
    GravityOverride:Vector="0,0,0"
    ComponentLinearAccScale:Vector="1,1,1"
    ComponentLinearVelScale:Vector="0,0,0"
    ExternalForce="0,0,0"
   ) -> (Pose) {
    alphaInputType=Bool
    angularDampingOverride=1
    angularSpringConstant=11000
    bAngularSpring=true
    bOverrideAngularDamping=true
    bUseGravityOverride=true
    gravityScale=0.35
    simulationSpace=RootRelative
    alphaBoolBlend {blendInTime=0.5, blendOption=QuadraticInOut, blendOutTime=1}
    boundBone {boneName=dyn_hair_24}
    componentAppliedLinearAccClamp=<3000 3000 3000>
    componentLinearAccScale=<1 1 1>
    componentLinearVelScale=<1 1 1>
    physicsBodyDefinitions [{
      boundBone {boneName=dyn_hair_24}
      boxExtents=<25 30 30>
      constraintSetup {angularLimitsMax=<1 6 8>, angularLimitsMin=<-1 -6 -8>, angularTarget=<1 0 0>}
      localJointOffset=<4 0 0>
     }]
    simSpaceSettings {maxAngularAcceleration=35, maxAngularVelocity=7, simSpaceAngularAlpha=1}
   }
   n11 = Anim Dynamics - Bone: dyn_hair_28!(
    ComponentPose=@n10
    bAlphaBoolEnabled:bool=True
    GravityOverride:Vector="0,0,-300"
    ComponentLinearAccScale:Vector="1,1,1"
    ComponentLinearVelScale:Vector="0,0,0"
    ExternalForce="0,0,0"
   ) -> (Pose) {
    alphaInputType=Bool
    angularDampingOverride=1
    angularSpringConstant=11000
    bAngularSpring=true
    bOverrideAngularDamping=true
    bUseGravityOverride=true
    gravityScale=0.35
    simulationSpace=RootRelative
    alphaBoolBlend {blendInTime=0.5, blendOption=QuadraticInOut, blendOutTime=1}
    boundBone {boneName=dyn_hair_28}
    componentAppliedLinearAccClamp=<3000 3000 3000>
    componentLinearAccScale=<1 1 1>
    componentLinearVelScale=<1 1 1>
    physicsBodyDefinitions [{
      boundBone {boneName=dyn_hair_28}
      boxExtents=<25 25 25>
      constraintSetup {angularLimitsMax=<1 20 -25>, angularLimitsMin=<-1 -20 30>, angularTarget=<1 0 0>}
      localJointOffset=<4 0 0>
     }]
    simSpaceSettings {maxAngularAcceleration=35, maxAngularVelocity=7, simSpaceAngularAlpha=1}
   }
   n12 = Anim Dynamics - Bone: dyn_hair_30!(
    ComponentPose=@n11
    bAlphaBoolEnabled:bool=True
    GravityOverride:Vector="0,0,-350"
    ComponentLinearAccScale:Vector="1,1,1"
    ComponentLinearVelScale:Vector="0,0,0"
    ExternalForce="0,0,0"
   ) -> (Pose) {
    alphaInputType=Bool
    angularDampingOverride=1
    angularSpringConstant=11000
    bAngularSpring=true
    bOverrideAngularDamping=true
    bUseGravityOverride=true
    gravityScale=0.35
    simulationSpace=RootRelative
    alphaBoolBlend {blendInTime=0.5, blendOption=QuadraticInOut, blendOutTime=1}
    boundBone {boneName=dyn_hair_30}
    componentAppliedLinearAccClamp=<3000 3000 3000>
    componentLinearAccScale=<1 1 1>
    componentLinearVelScale=<1 1 1>
    physicsBodyDefinitions [{
      boundBone {boneName=dyn_hair_30}
      boxExtents=<20 30 30>
      constraintSetup {angularLimitsMax=<1 5 6>, angularLimitsMin=<-1 -5 -6>, angularTarget=<1 0 0>}
      localJointOffset=<4 0 0>
     }]
    simSpaceSettings {maxAngularAcceleration=35, maxAngularVelocity=7, simSpaceAngularAlpha=1}
   }
   n13 = Layered blend per bone(
    BasePose=Component To Local(ComponentPose=@n12)
    BlendPoses_0=Slot 'DefaultSlot'()
    BlendWeights_0=Get Custom Idle Is Playing Float() -> (Custom Idle Is Playing Float:double)
   ) {layerSetup [{branchFilters [
       {boneName=dyn_hair_02}
       {boneName=dyn_hair_05}
       {boneName=dyn_hair_08}
       {boneName=dyn_hair_09}
       {boneName=dyn_hair_10}
       {boneName=dyn_hair_13}
       {boneName=dyn_hair_14}
       {boneName=dyn_hair_16}
       {boneName=dyn_hair_20}
       {boneName=dyn_hair_24}
       {boneName=dyn_hair_28}
       {boneName=dyn_hair_30}
       {boneName=base_hair_02}
       {boneName=base_hair_05}
       {boneName=base_hair_08}
       {boneName=base_hair_09}
       {boneName=base_hair_10}
       {boneName=base_hair_13}
       {boneName=base_hair_14}
       {boneName=base_hair_16}
       {boneName=base_hair_20}
       {boneName=base_hair_24}
       {boneName=base_hair_28}
       {boneName=base_hair_30}
      ]}]}
   Output Pose(Result=@n13) {layerGroup=None}
  ]}
 EventGraph {nodes [
   Event BlueprintUpdateAnimation()
   > n1 = Sequence()
   n2 = Event BlueprintProcessAnimInputEvent()
   n3 = IsNamed(self=@n2.AnimInputEvent, InEventName=aie.Player.Zipline.Entered)
   n4 = OR Boolean(A=Get bIsMoving2D(), B=Get bIsTacticalSprinting())
   n5 = Get bIsJumping()
   n6 = Branch(execute=@n1.then_1, Condition=OR Boolean(A=@n3, B=AND Boolean!(A=@n3, B=@n5, C:bool=@n4) -> (ReturnValue)))
   > Set Componen Linear Vel Scale(Componen Linear Vel Scale:Vector="0, 0, 0") -> (Output_Get:Vector)
   Branch(execute=@n6.else, Condition=IsNamed(self=@n2.AnimInputEvent, InEventName=aie.Player.Zipline.Exited))
   > Set Componen Linear Vel Scale(Componen Linear Vel Scale="0.3,0.3,0.3")
   n7 = "Gravity Override State Function (Parameter Asset) (Nativized)"(
    execute=@n1.then_0
    CharacterPartAnimInstance=Self-Reference()
    GameTimeInSeconds=GetGameTimeInSeconds()
    Parameters=/BRCosmetics/Characters/Player/Male/Medium/Bodies/M_MED_RampCanal/Meshes/M_MED_RampCanal_Parameters
   )
   > Set Gravity Override(Gravity Override=@n7)
   > Branch(Condition=Get bIsInFrontEnd())
   > Set Custom Idle Is Playing Float(Custom Idle Is Playing Float:double=GetCurveValue(self=GetOwnerFrontEndAnimInstance(), CurveName=Frontend_Idle_Bypass)) -> (Output_Get:double)
  ]}
]
variables [
 Componen Linear Vel Scale {type=Vector, default=<0.3 0.3 0.3>}
 Custom Idle Is Playing Float {type=number}
 Gravity Override {type=Vector}
]
